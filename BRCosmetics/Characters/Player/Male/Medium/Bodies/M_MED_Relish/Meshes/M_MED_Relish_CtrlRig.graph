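# Control Rig for the M_MED_Relish male character body that procedurally corrects the left shoulder strap bone as the arm moves.
# It reads upperarm_l's rotation relative to spine_05, converts it to ZYX Euler and takes the absolute Z angle.
# A clamped, inverted remap turns that angle into the weight of a fixed translation offset on dyn_l_shoulder_strap, growing as the angle decreases.
# The hierarchy holds many other dyn and deform bones, but the graph modifies only this one.
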ControlRigBlueprint /BRCosmetics/Characters/Player/Male/Medium/Bodies/M_MED_Relish/Meshes/M_MED_Relish_CtrlRig
graphs [
 RigVMFunctionLibrary {links [], nodes []}
 RigVMModel {nodes [
   RigUnit_BeginExecution = &RigUnit_BeginExecution()
   n1 = &RigVMFunction_MathQuaternionToEuler(
    RotationOrder=ZYX
    Value=&RigUnit_GetRelativeTransformForItem(Child=(Type=Bone,Name="upperarm_l"), Parent=(Type=Bone,Name="spine_05")).RelativeTransform.Rotation
   )
   &RigUnit_ModifyTransforms(
    ItemToModify=((Item=(Type=Bone,Name="dyn_l_shoulder_strap"),Transform=(Rotation=(X=0,Y=0,Z=0,W=1),Translation=(X=5,Y=1,Z=2),Scale3D=(X=1,Y=1,Z=1))))
    Weight=&RigVMFunction_MathFloatRemap(
     SourceMaximum=50
     SourceMinimum=170
     TargetMinimum=0
     bClamp=true
     Value=/Script/RigVM.RigVMFunction_MathFloatAbs(Value=@n1.Result.Z) {is=struct RigVMUnitNode, template="Absolute::Execute(in Value,out Result)"}.Result
    ).Result
    ExecutePin=@RigUnit_BeginExecution.ExecutePin
   )
  ]}
]
hierarchy [
 root {loc=<0 0 0>, rot=<0 0 0 1>}
 pelvis {parent=root, loc=<0 -0.7071 93.1191>, rot=<0.0127 0.707 -0.0127 -0.707>}
 spine_01 {parent=pelvis, loc=<6.7494 -0.6388 0>, rot=<0 0 0.034 -0.9994>}
 spine_02 {parent=spine_01, loc=<7.2518 -0.035 0>, rot=<0 0 0.0238 -0.9997>}
 spine_03 {parent=spine_02, loc=<6.0104 0 0>, rot=<0 0 0.0149 -0.9999>}
 spine_04 {parent=spine_03, loc=<8.1024 0 0>, rot=<0 0 -0.0132 -0.9999>}
 spine_05 {parent=spine_04, loc=<10.0146 0 0>, rot=<0 0 -0.0015 -1>}
 clavicle_l {parent=spine_05, loc=<4.3575 -0.1466 -2.3223>, rot=<0.0465 -0.7289 -0.009 -0.6829>}
 upperarm_l {parent=clavicle_l, loc=<15.5111 0 0>, rot=<0.0226 -0.3234 -0.0088 -0.9459>}
 lowerarm_l {parent=upperarm_l, loc=<26.6514 0 0>, rot=<0.0391 -0.0099 -0.2453 -0.9686>}
 hand_l {parent=lowerarm_l, loc=<22.3491 0 0>, rot=<-0.7081 -0.0507 0.0975 -0.6975>}
 index_metacarpal_l {parent=hand_l, loc=<1.6535 0.0183 -2.96>, rot=<0.0158 -0.0124 -0.0173 -0.9996>}
 index_01_l {parent=index_metacarpal_l, loc=<8.3008 0.0357 -0.0524>, rot=<-0.0044 0.0485 0.0924 -0.9945>}
 index_02_l {parent=index_01_l, loc=<3.3587 -0.3004 0.0452>, rot=<-0.002 0.0027 0.3225 -0.9466>}
 index_03_l {parent=index_02_l, loc=<2.5578 -0.123 -0.0252>, rot=<-0.002 0.0067 0.2558 -0.9667>}
 middle_metacarpal_l {parent=hand_l, loc=<2.1856 -0.1725 -0.5984>, rot=<0.006 0.0168 -0.0329 -0.9993>}
 middle_01_l {parent=middle_metacarpal_l, loc=<7.3275 -0.2447 0.0129>, rot=<-0.0208 0.0042 0.0858 -0.9961>}
 middle_02_l {parent=middle_01_l, loc=<4.0123 0.1401 0.1175>, rot=<-0.0218 0.0258 0.3512 -0.9357>}
 middle_03_l {parent=middle_02_l, loc=<2.6077 0.0534 0.1373>, rot=<0.0082 0.0188 0.1708 -0.9851>}
 pinky_metacarpal_l {parent=hand_l, loc=<1.7616 -0.9753 3.3814>, rot=<-0.0397 0.0814 -0.069 -0.9935>}
 pinky_01_l {parent=pinky_metacarpal_l, loc=<7.1709 -0.4831 -0.1094>, rot=<-0.0654 -0.0035 0.2242 -0.9723>}
 pinky_02_l {parent=pinky_01_l, loc=<2.8274 0.0893 -0.1043>, rot=<0.0037 0.0011 0.2854 -0.9584>}
 pinky_03_l {parent=pinky_02_l, loc=<2.2046 0.0307 -0.0078>, rot=<-0.0002 0.0098 0.1445 -0.9895>}
 ring_metacarpal_l {parent=hand_l, loc=<2.2074 -0.3625 1.3638>, rot=<-0.0284 0.0722 -0.0458 -0.9959>}
 ring_01_l {parent=ring_metacarpal_l, loc=<7.0295 -0.6197 -0.1446>, rot=<-0.0489 -0.0252 0.1892 -0.9804>}
 ring_02_l {parent=ring_01_l, loc=<3.6174 0.4647 -0.1027>, rot=<0 0 0.2806 -0.9598>}
 ring_03_l {parent=ring_02_l, loc=<2.4152 0.114 0.0199>, rot=<0 0 0.1189 -0.9929>}
 thumb_01_l {parent=hand_l, loc=<1.3118 -0.3191 -2.459>, rot=<0.6546 -0.4042 0.0664 -0.6354>}
 thumb_02_l {parent=thumb_01_l, loc=<5.9336 -0.0448 0.0401>, rot=<-0.0037 0.0213 0.298 -0.9543>}
 thumb_03_l {parent=thumb_02_l, loc=<2.2223 0.3682 -0.0907>, rot=<0.0118 -0.013 0.0066 -0.9998>}
 lowerarm_twist_01_l {parent=lowerarm_l, loc=<17.8549 0 0>, rot=<0 0 0 -1>}
 lowerarm_twist_02_l {parent=lowerarm_l, loc=<9.6199 0 0>, rot=<0 0 0 -1>}
 deform_elbow_bk_l {parent=lowerarm_l, loc=<-1.2853 -2.9274 -0.5607>, rot=<-0.0285 0.0335 -0.0024 -0.999>}
 upperarm_twist_01_l {parent=upperarm_l, loc=<0.0041 0 0>, rot=<0 0 0 -1>}
 deform_deltoid_fr_l {parent=upperarm_twist_01_l, loc=<2.4511 0 0>, rot=<0 0 0 -1>}
 upperarm_twist_02_l {parent=upperarm_l, loc=<13.7941 0 0>, rot=<0 0 0 -1>}
 deform_elbow_fr_l {parent=upperarm_l, loc=<26.5048 0.1875 -0.5816>, rot=<0.0021 0.0156 -0.185 -0.9826>}
 dyn_radio_base_master {parent=clavicle_l, loc=<7.3305 13.4462 2.2953>, rot=<0.2747 -0.3829 -0.6699 -0.5737>}
 dyn_radio_base {parent=dyn_radio_base_master, loc=<0 0 0>, rot=<0.077 0 0 -0.997>}
 dyn_radio_cable_01 {parent=dyn_radio_base, loc=<2.2794 0 0>, rot=<0.0023 0.0093 -0.0289 -0.9995>}
 dyn_radio_cable_02 {parent=dyn_radio_cable_01, loc=<4.2223 0 0>, rot=<-0.0925 -0.3455 0.0132 -0.9338>}
 dyn_radio_cable_03 {parent=dyn_radio_cable_02, loc=<4.4232 0 0>, rot=<-0.0069 -0.1233 0.1891 -0.9742>}
 dyn_radio_cable_04 {parent=dyn_radio_cable_03, loc=<5.0943 0 0>, rot=<0.0132 -0.0459 0.1432 -0.9885>}
 dyn_radio_cable_05 {parent=dyn_radio_cable_04, loc=<5.5165 0 0>, rot=<0.0926 -0.1838 0.0422 -0.9777>}
 dyn_radio_cable_06 {parent=dyn_radio_cable_05, loc=<5.4539 0 0>, rot=<0.0513 -0.1451 -0.1671 -0.9738>}
 dyn_radio_cable_07 {parent=dyn_radio_cable_06, loc=<6.4443 0 0>, rot=<0.0015 -0.0345 -0.1272 -0.9913>}
 dyn_radio_cable_08 {parent=dyn_radio_cable_07, loc=<6.563 0 0>, rot=<0.0716 0.2874 -0.1509 -0.9431>}
 dyn_radio_cable_09 {parent=dyn_radio_cable_08, loc=<3.7055 0 0>, rot=<0.0478 0.2127 0.1764 -0.9599>}
 dyn_radio_cable_10 {parent=dyn_radio_cable_09, loc=<3.4319 0 0>, rot=<0.9489 -0.0013 0.0757 0.3063>}
 deform_pectoral_fr_l {parent=clavicle_l, loc=<9.3905 0 0>, rot=<0 0 0 1>}
 dyn_l_shoulder_strap {parent=clavicle_l, loc=<9.3905 0 0>, rot=<0 0 0 1>}
 clavicle_r {parent=spine_05, loc=<4.3575 -0.1466 2.3223>, rot=<0.7289 0.0465 -0.6829 0.009>}
 upperarm_r {parent=clavicle_r, loc=<-15.5111 0 0>, rot=<0.0226 -0.3234 -0.0088 -0.9459>}
 lowerarm_r {parent=upperarm_r, loc=<-26.6515 0 0>, rot=<0.0391 -0.0099 -0.2453 -0.9686>}
 hand_r {parent=lowerarm_r, loc=<-22.3492 0 0>, rot=<-0.7081 -0.0507 0.0975 -0.6975>}
 index_metacarpal_r {parent=hand_r, loc=<-1.6534 -0.0181 2.9601>, rot=<0.0158 -0.0124 -0.0173 -0.9996>}
 index_01_r {parent=index_metacarpal_r, loc=<-8.3008 -0.0357 0.0524>, rot=<-0.0044 0.0485 0.0924 -0.9945>}
 index_02_r {parent=index_01_r, loc=<-3.3589 0.3003 -0.0453>, rot=<-0.002 0.0027 0.3225 -0.9466>}
 index_03_r {parent=index_02_r, loc=<-2.5578 0.1231 0.0251>, rot=<-0.002 0.0067 0.2558 -0.9667>}
 middle_metacarpal_r {parent=hand_r, loc=<-2.1856 0.1726 0.5984>, rot=<0.006 0.0168 -0.0329 -0.9993>}
 middle_01_r {parent=middle_metacarpal_r, loc=<-7.3273 0.2447 -0.0129>, rot=<-0.0208 0.0042 0.0858 -0.9961>}
 middle_02_r {parent=middle_01_r, loc=<-4.0142 -0.1411 -0.07>, rot=<-0.0217 0.0259 0.3512 -0.9357>}
 middle_03_r {parent=middle_02_r, loc=<-2.6074 -0.0534 -0.1312>, rot=<0.0083 0.0188 0.1708 -0.9851>}
 pinky_metacarpal_r {parent=hand_r, loc=<-1.7617 0.9753 -3.3814>, rot=<-0.0397 0.0814 -0.069 -0.9935>}
 pinky_01_r {parent=pinky_metacarpal_r, loc=<-7.1707 0.4832 0.1094>, rot=<-0.0654 -0.0035 0.2242 -0.9723>}
 pinky_02_r {parent=pinky_01_r, loc=<-2.8274 -0.0892 0.1043>, rot=<0.0037 0.0011 0.2854 -0.9584>}
 pinky_03_r {parent=pinky_02_r, loc=<-2.2045 -0.031 0.0078>, rot=<-0.0002 0.0098 0.1445 -0.9895>}
 ring_metacarpal_r {parent=hand_r, loc=<-2.2073 0.3627 -1.3638>, rot=<-0.0284 0.0722 -0.0458 -0.9959>}
 ring_01_r {parent=ring_metacarpal_r, loc=<-7.0297 0.6197 0.1446>, rot=<-0.0489 -0.0252 0.1892 -0.9804>}
 ring_02_r {parent=ring_01_r, loc=<-3.6171 -0.4648 0.1027>, rot=<0 0 0.2806 -0.9598>}
 ring_03_r {parent=ring_02_r, loc=<-2.4153 -0.1141 -0.0199>, rot=<0 0 0.1189 -0.9929>}
 thumb_01_r {parent=hand_r, loc=<-1.3119 0.3192 2.459>, rot=<0.6546 -0.4042 0.0664 -0.6354>}
 thumb_02_r {parent=thumb_01_r, loc=<-5.9336 0.0447 -0.0401>, rot=<-0.0037 0.0213 0.298 -0.9543>}
 thumb_03_r {parent=thumb_02_r, loc=<-2.2223 -0.3682 0.0907>, rot=<0.0118 -0.013 0.0066 -0.9998>}
 lowerarm_twist_01_r {parent=lowerarm_r, loc=<-17.855 0 0>, rot=<0 0 0 1>}
 lowerarm_twist_02_r {parent=lowerarm_r, loc=<-9.6199 0 0>, rot=<0 0 0 1>}
 deform_elbow_bk_r {parent=lowerarm_r, loc=<1.2852 2.9274 0.561>, rot=<-0.0285 0.0335 -0.0024 -0.999>}
 upperarm_twist_01_r {parent=upperarm_r, loc=<-0.0041 0 0>, rot=<0 0 0 1>}
 deform_deltoid_fr_r {parent=upperarm_twist_01_r, loc=<-2.4512 0 0>, rot=<0 0 0 1>}
 upperarm_twist_02_r {parent=upperarm_r, loc=<-13.7942 0 0>, rot=<0 0 0 1>}
 deform_elbow_fr_r {parent=upperarm_r, loc=<-26.5046 -0.1875 0.5813>, rot=<0.0021 0.0156 -0.185 -0.9826>}
 dyn_clip_r_shoulder_01_master {parent=clavicle_r, loc=<-7.5601 -6.4723 -8.2381>, rot=<-0.5996 -0.6743 0.3359 -0.2701>}
 dyn_clip_r_shoulder_01 {parent=dyn_clip_r_shoulder_01_master, loc=<0 0 0>, rot=<0 0 0 1>}
 dyn_clip_r_shoulder_02_master {parent=clavicle_r, loc=<-7.5476 -7.9273 -6.9431>, rot=<0.6946 -0.6509 0.199 0.2328>}
 dyn_clip_r_shoulder_02 {parent=dyn_clip_r_shoulder_02_master, loc=<0 0 0>, rot=<0 0 0 1>}
 dyn_clip_l_shoulder_01_master {parent=clavicle_r, loc=<13.9753 -5.5616 -6.9601>, rot=<0.6777 0.6404 -0.1811 0.3128>}
 dyn_clip_l_shoulder_01 {parent=dyn_clip_l_shoulder_01_master, loc=<0 0 0>, rot=<0 0 0 1>}
 dyn_clip_l_shoulder_02_master {parent=clavicle_r, loc=<13.7604 -7.013 -5.6176>, rot=<0.6772 -0.6676 0.2704 0.1502>}
 dyn_clip_l_shoulder_02 {parent=dyn_clip_l_shoulder_02_master, loc=<0 0 0>, rot=<0 0 0 -1>}
 deform_pectoral_fr_r {parent=clavicle_r, loc=<-9.39 0 0>, rot=<0 0 0 1>}
 neck_01 {parent=spine_05, loc=<11.3138 0.1529 0>, rot=<0 0 -0.2003 -0.9797>}
 neck_02 {parent=neck_01, loc=<4.5627 0 0>, rot=<0 0 0.0504 -0.9987>}
 head {parent=neck_02, loc=<5.3914 0 0>, rot=<0 0 0.115 -0.9934>}
 dyn_neck {parent=neck_01, loc=<4.5627 0 0>, rot=<0 0 0.0504 -0.9987>}
 dyn_clip_shoulder_r_back_02_master {parent=spine_05, loc=<1.955 -13.7174 8.9956>, rot=<-0.0775 -0.7376 0.6661 -0.0794>}
 dyn_clip_shoulder_r_back_02 {parent=dyn_clip_shoulder_r_back_02_master, loc=<0 0 0>, rot=<0 0 0 1>}
 dyn_clip_shoulder_l_back_02_master {parent=spine_05, loc=<2.0051 -13.8848 -8.8684>, rot=<0.0496 -0.7283 0.6742 0.1117>}
 dyn_clip_shoulder_l_back_02 {parent=dyn_clip_shoulder_l_back_02_master, loc=<0 0 0>, rot=<0 0 0 -1>}
 dyn_clip_shoulder_r_back_01_master {parent=spine_05, loc=<9.9927 -10.3247 10.5764>, rot=<0.2197 -0.6367 0.6667 -0.319>}
 dyn_clip_shoulder_r_back_01 {parent=dyn_clip_shoulder_r_back_01_master, loc=<0 0 0>, rot=<0 0 0 1>}
 dyn_clip_shoulder_l_back_01_master {parent=spine_05, loc=<10.0096 -10.3174 -10.6175>, rot=<0.2942 -0.6738 0.6606 -0.1517>}
 dyn_clip_shoulder_l_back_01 {parent=dyn_clip_shoulder_l_back_01_master, loc=<0 0 0>, rot=<0 0 0 1>}
 dyn_zipper_master {parent=spine_05, loc=<6.4151 10.1291 -0.0058>, rot=<0.2199 0.672 0.6734 0.2158>}
 dyn_zipper {parent=dyn_zipper_master, loc=<0 0 0>, rot=<0 0 0 1>}
 dyn_insideCollar_Fix {parent=spine_05, loc=<9.9916 7.2683 -4.6955>, rot=<0.0465 -0.7289 -0.009 -0.6829>}
 dyn_radio_top_cable_05 {parent=spine_05, loc=<15.291 1.407 -9.8063>, rot=<0.0284 -0.7065 -0.0284 -0.7065>}
 dyn_radio_top_cable_lookat {parent=dyn_radio_top_cable_05, loc=<0 0 0>, rot=<0.2272 -0.2178 -0.6569 -0.6852>}
 dyn_radio_top_cable_04 {parent=dyn_radio_top_cable_lookat, loc=<2.7464 0.0226 0.9864>, rot=<-0.1715 0.0452 0.9841 0.004>}
 dyn_radio_top_cable_03 {parent=dyn_radio_top_cable_04, loc=<-5.0399 0.9571 -0.673>, rot=<-0.1954 0.0825 0.0784 -0.9741>}
 dyn_radio_top_cable_02 {parent=dyn_radio_top_cable_03, loc=<-2.5202 -0.8178 -0.4983>, rot=<0.1964 0.1229 -0.1323 -0.9637>}
 dyn_radio_top_cable_01 {parent=dyn_radio_top_cable_02, loc=<-3.3314 -1.1081 -0.7614>, rot=<0.125 0.1271 -0.1421 -0.9737>}
 dyn_clip_back_r_01_master {parent=spine_03, loc=<1.4641 -9.4335 7.4313>, rot=<-0.4212 0.5861 -0.4622 -0.5153>}
 dyn_clip_back_r_01 {parent=dyn_clip_back_r_01_master, loc=<0 0 0>, rot=<0 0 0 -1>}
 dyn_clip_back_l_01_master {parent=spine_03, loc=<1.1993 -9.4131 -7.2611>, rot=<-0.5384 -0.4584 0.5563 -0.4365>}
 dyn_clip_back_l_01 {parent=dyn_clip_back_l_01_master, loc=<0 0 0>, rot=<0 0 0 -1>}
 dyn_underarm_strap_l_master {parent=spine_02, loc=<6.652 -1.1925 -13.9813>, rot=<0.6676 -0.7209 -0.1267 0.1362>}
 dyn_underarm_strap_l {parent=dyn_underarm_strap_l_master, loc=<0 0 0>, rot=<0 0 0 1>}
 dyn_underarm_strap_r_master {parent=spine_02, loc=<6.4007 -1.3637 13.8821>, rot=<-0.123 0.1337 0.7045 -0.6861>}
 dyn_underarm_strap_r {parent=dyn_underarm_strap_r_master, loc=<0 0 0>, rot=<0 0 0 1>}
 dyn_underarm_clip_r_master {parent=spine_02, loc=<4.5769 0.2483 13.553>, rot=<0.0218 0.0007 0.9998 -0.0035>}
 dyn_underarm_clip_r {parent=dyn_underarm_clip_r_master, loc=<0 0 0>, rot=<0 0 0 -1>}
 dyn_underarm_clip_l_master {parent=spine_02, loc=<4.5277 0.292 -13.5605>, rot=<-0.0065 -0.9989 -0.0348 -0.0317>}
 dyn_underarm_clip_l {parent=dyn_underarm_clip_l_master, loc=<0 0 0>, rot=<0 0 0 -1>}
 dyn_clip_back_l_02_master {parent=spine_02, loc=<2.7254 -8.7223 -7.0987>, rot=<-0.4837 -0.5158 0.5171 -0.4823>}
 dyn_clip_back_l_02 {parent=dyn_clip_back_l_02_master, loc=<0 0 0>, rot=<0 0 0 1>}
 dyn_clip_back_r_02_master {parent=spine_02, loc=<2.7575 -8.8063 7.0966>, rot=<0.4558 -0.549 0.5223 0.4669>}
 dyn_clip_back_r_02 {parent=dyn_clip_back_r_02_master, loc=<0 0 0>, rot=<0 0 0 1>}
 thigh_l {parent=pelvis, loc=<-2.5693 -0.0109 -7.7501>, rot=<0.0012 0.0572 0.0155 -0.9982>}
 calf_l {parent=thigh_l, loc=<-39.0484 0 0>, rot=<0.0006 -0.0161 -0.0391 -0.9991>}
 calf_twist_01_l {parent=calf_l, loc=<-35.813 -0.0967 0.0397>, rot=<0.0237 0.0095 -0.0035 -0.9997>}
 calf_twist_02_l {parent=calf_l, loc=<-18.8064 -0.0967 0.0397>, rot=<0.0237 0.0095 -0.0035 -0.9997>}
 foot_l {parent=calf_l, loc=<-42.561 0.1077 0.4292>, rot=<0.0418 -0.0455 0.0343 -0.9975>}
 ball_l {parent=foot_l, loc=<-6.1599 13.8264 0.6806>, rot=<0.0001 -0.0001 -0.7063 -0.7079>}
 deform_knee_bk_01_l {parent=calf_l, loc=<0.6022 -1.0607 0>, rot=<0 0 -0.0408 -0.9992>}
 deform_knee_fr_01_l {parent=calf_l, loc=<-0.2342 0 0>, rot=<0 0 0 1>}
 dyn_l_kneeFold {parent=calf_l, loc=<0 0 0>, rot=<0 0 0 1>}
 thigh_twist_01_l {parent=thigh_l, loc=<-21.6303 0 0>, rot=<0.0013 0 0 -1>}
 deform_thigh_fr_l {parent=thigh_twist_01_l, loc=<21.5084 0.2281 -0.0598>, rot=<-0.0013 -0.0217 0 -0.9998>}
 deform_thigh_bk_l {parent=thigh_twist_01_l, loc=<16.9514 0.2058 0.1808>, rot=<-0.0013 -0.0217 0 -0.9998>}
 thigh_r {parent=pelvis, loc=<-2.5691 -0.0109 7.7501>, rot=<0.0572 -0.0012 0.9982 0.0155>}
 calf_r {parent=thigh_r, loc=<39.0487 0 0>, rot=<0.0006 -0.0161 -0.0391 -0.9991>}
 calf_twist_01_r {parent=calf_r, loc=<35.8132 0.0967 -0.0397>, rot=<0.0237 0.0095 -0.0035 -0.9997>}
 calf_twist_02_r {parent=calf_r, loc=<18.8064 0.0967 -0.0397>, rot=<0.0237 0.0095 -0.0035 -0.9997>}
 foot_r {parent=calf_r, loc=<42.5612 -0.1077 -0.4292>, rot=<0.0418 -0.0455 0.0343 -0.9975>}
 ball_r {parent=foot_r, loc=<6.1598 -13.8264 -0.6806>, rot=<0.0001 -0.0001 -0.7063 -0.7079>}
 deform_knee_bk_01_r {parent=calf_r, loc=<-0.6022 1.0607 0>, rot=<0 0 -0.0408 -0.9992>}
 deform_knee_fr_01_r {parent=calf_r, loc=<0.2341 0 0>, rot=<0 0 0 1>}
 dyn_r_kneeFold {parent=calf_r, loc=<0 0 0>, rot=<0 0 0 1>}
 thigh_twist_01_r {parent=thigh_r, loc=<21.6303 0 0>, rot=<0.0013 0 0 -1>}
 deform_thigh_fr_r {parent=thigh_twist_01_r, loc=<-21.5083 -0.2281 0.0598>, rot=<-0.0013 -0.0217 0 -0.9998>}
 deform_thigh_bk_r {parent=thigh_twist_01_r, loc=<-16.9512 -0.2058 -0.1808>, rot=<-0.0013 -0.0217 0 -0.9998>}
 dyn_holster_master {parent=thigh_r, loc=<14.7585 -0.8833 9.9451>, rot=<-0.0002 -0.0572 0.0024 -0.9984>}
 dyn_holster {parent=dyn_holster_master, loc=<0 0 0>, rot=<0 0 0 1>}
 dyn_pouch_back_large_master {parent=pelvis, loc=<5.6817 -12.1792 0>, rot=<0.0468 -0.7056 0.7056 -0.0468>}
 dyn_pouch_back_large {parent=dyn_pouch_back_large_master, loc=<0 0 0>, rot=<0 0 0 -1>}
 dyn_pouch_back_large_strap {parent=dyn_pouch_back_large, loc=<4.4697 0 6.113>, rot=<0 0 0 -1>}
 dyn_pouch_04_master {parent=pelvis, loc=<5.8269 -10.0398 10.3027>, rot=<0.0874 -0.5016 0.8601 -0.0321>}
 dyn_pouch_04 {parent=dyn_pouch_04_master, loc=<0 0 0>, rot=<0 0 0 1>}
 dyn_pouch_04_strap {parent=dyn_pouch_04, loc=<3.2971 -0.0222 3.0267>, rot=<-0.1039 0.0837 0.0052 -0.991>}
 dyn_pouch_01_master {parent=pelvis, loc=<5.5303 9.5549 -10.9451>, rot=<-0.0024 -0.819 -0.5736 -0.0146>}
 dyn_pouch_01 {parent=dyn_pouch_01_master, loc=<0 0 0>, rot=<0 0 0 1>}
 dyn_pouch_02_master {parent=pelvis, loc=<6.517 6.2967 -14.7042>, rot=<0.0079 -0.9286 -0.3689 -0.0386>}
 dyn_pouch_02 {parent=dyn_pouch_02_master, loc=<0 0 0>, rot=<0 0 0 1>}
 dyn_pouch_02_strap {parent=dyn_pouch_02, loc=<3.0024 -0.537 2.2814>, rot=<-0.0738 0.2193 0.011 -0.9728>}
 dyn_pouch_03_master {parent=pelvis, loc=<7.0765 1.8823 -16.0365>, rot=<-0.0439 -0.9922 -0.0871 -0.0782>}
 dyn_pouch_03 {parent=dyn_pouch_03_master, loc=<0 0 0>, rot=<0 0 0 1>}
 dyn_pouch_03_strap {parent=dyn_pouch_03, loc=<3.9566 -0.2792 2.3798>, rot=<-0.0001 0.1925 0.0033 -0.9813>}
 dyn_pouch_05_master {parent=pelvis, loc=<5.7848 -4.7131 14.8998>, rot=<0.0558 -0.2503 0.9652 -0.052>}
 dyn_pouch_05 {parent=dyn_pouch_05_master, loc=<0 0 0>, rot=<0 0 0 1>}
 dyn_pouch_05_strap {parent=dyn_pouch_05, loc=<3.457 0.2659 2.9423>, rot=<-0.0307 0.0952 0.0145 -0.9949>}
 dyn_pouch_radio_master {parent=pelvis, loc=<7.6222 -9.8926 -10.8118>, rot=<0.0322 -0.8908 0.4453 -0.0849>}
 dyn_pouch_radio {parent=dyn_pouch_radio_master, loc=<0 0 0>, rot=<0 0 0 1>}
 dyn_pouch_radio_wire_base {parent=dyn_pouch_radio, loc=<-3.3308 1.6161 1.013>, rot=<0.2797 0.5637 -0.3478 -0.695>}
]
