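# Physics asset for the M_MED_Relish male medium player body, covering a radio prop with its dangling cable plus the torso and left arm.
# Capsules form a pelvis-to-upper-spine chain that branches into the left arm and a box-shaped radio base.
# An eight-capsule cable hangs from the radio on an angularly free first joint, with tight swing and twist limits between later links keeping it fairly stiff.
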PhysicsAsset /BRCosmetics/Characters/Player/Male/Medium/Bodies/M_MED_Relish/Meshes/M_MED_Relish_Physics
bodies {
 dyn_radio_base [{
   shapeType=Box
   extentX=3.9629
   extentY=2.4552
   extentZ=10.5874
   shapeName=dyn_radio_base_box
   center=<-0.1389 -0.1338 0.0626>
   rotation=<rot 0 -91.4406 -88.8133>
  }]
 dyn_radio_cable_01 [{shapeType=Capsule, length=3.8264, radius=0.505, shapeName=dyn_radio_cable_01_capsule, center=<3 0 0>, rotation=<rot 90 0 0>}]
 dyn_radio_cable_02 [{shapeType=Capsule, length=3.8264, radius=0.505, shapeName=dyn_radio_cable_02_capsule, center=<3 0 0>, rotation=<rot 90 0 0>}]
 dyn_radio_cable_03 [{shapeType=Capsule, length=3.8264, radius=0.505, shapeName=dyn_radio_cable_03_capsule, center=<3 0 0>, rotation=<rot 90 0 0>}]
 dyn_radio_cable_04 [{shapeType=Capsule, length=3.8264, radius=0.505, shapeName=dyn_radio_cable_04_capsule, center=<3 0 0>, rotation=<rot 90 0 0>}]
 dyn_radio_cable_05 [{shapeType=Capsule, length=3.8264, radius=0.505, shapeName=dyn_radio_cable_05_capsule, center=<3 0 0>, rotation=<rot 90 0 0>}]
 dyn_radio_cable_06 [{shapeType=Capsule, length=3.8264, radius=0.505, shapeName=dyn_radio_cable_06_capsule, center=<3 0 0>, rotation=<rot 90 0 0>}]
 dyn_radio_cable_07 [{shapeType=Capsule, length=3.8264, radius=0.505, shapeName=dyn_radio_cable_07_capsule, center=<3 0 0>, rotation=<rot 90 0 0>}]
 dyn_radio_cable_08 [{shapeType=Capsule, length=3.8264, radius=0.505, shapeName=dyn_radio_cable_08_capsule, center=<3 0 0>, rotation=<rot 90 0 0>}]
 hand_l [{shapeType=Capsule, length=8.9362, radius=6.6571, center=<7.6413 -2.8158 0.0534>, rotation=<rot 0 89.2872 -66.1886>}]
 lowerarm_l [{shapeType=Capsule, length=14.8982, radius=4.9143, center=<10.3555 0.873 0.8097>, rotation=<rot 0 90.5993 -86.09>}]
 pelvis [{shapeType=Capsule, length=8.8487, radius=12.6667}]
 spine_01 [{shapeType=Capsule, length=11.0332, radius=8.5483, shapeName=spine_01_capsule, center=<0 1.6548 0>}]
 spine_02 [{shapeType=Capsule, length=8.959, radius=9.0551, center=<0 1.3659 0>}]
 spine_03 [{shapeType=Capsule, length=8.4577, radius=9.5005, center=<0 1.8647 0>}]
 spine_04 [{shapeType=Capsule, length=4.3092, radius=13.5051}]
 spine_05 [{shapeType=Capsule, length=12.681, radius=12.9554, center=<-1.7913 1.3149 0>}]
 upperarm_l [{shapeType=Capsule, length=18.6484, radius=6.2659, center=<11.4247 -1.119 -0.1418>, rotation=<rot 0 -96.1904 -85.0371>}]
}
constraints [
 {bone1Name=spine_02, bone2Name=spine_01, swing1LimitDegrees=45, swing2LimitDegrees=45, twistLimitDegrees=45}
 {bone1Name=spine_03, bone2Name=spine_02, swing1LimitDegrees=45, swing2LimitDegrees=45, twistLimitDegrees=45}
 {bone1Name=spine_04, bone2Name=spine_03, swing1LimitDegrees=45, swing2LimitDegrees=45, twistLimitDegrees=45}
 {bone1Name=spine_05, bone2Name=spine_04, swing1LimitDegrees=45, swing2LimitDegrees=45, twistLimitDegrees=45}
 {bone1Name=upperarm_l, bone2Name=spine_05, swing1LimitDegrees=45, swing2LimitDegrees=45, twistLimitDegrees=45}
 {bone1Name=lowerarm_l, bone2Name=upperarm_l, swing1LimitDegrees=45, swing2LimitDegrees=45, twistLimitDegrees=45}
 {bone1Name=hand_l, bone2Name=lowerarm_l, swing1LimitDegrees=45, swing2LimitDegrees=45, twistLimitDegrees=45}
 {bone1Name=dyn_radio_base, bone2Name=spine_05, swing1LimitDegrees=45, swing2LimitDegrees=45, twistLimitDegrees=45}
 {
  bone1Name=spine_01
  bone2Name=pelvis
  swing1LimitDegrees=45
  swing1Motion=Free
  swing2LimitDegrees=45
  swing2Motion=Free
  twistLimitDegrees=45
  twistMotion=Free
 }
 {
  bone1Name=dyn_radio_cable_01
  bone2Name=dyn_radio_base
  swing1LimitDegrees=45
  swing1Motion=Free
  swing2LimitDegrees=45
  swing2Motion=Free
  twistLimitDegrees=45
  twistMotion=Free
 }
 {bone1Name=dyn_radio_cable_02, bone2Name=dyn_radio_cable_01, swing1LimitDegrees=3, swing2LimitDegrees=3, twistLimitDegrees=1}
 {bone1Name=dyn_radio_cable_03, bone2Name=dyn_radio_cable_02, swing1LimitDegrees=5, swing2LimitDegrees=5, twistLimitDegrees=3}
 {bone1Name=dyn_radio_cable_04, bone2Name=dyn_radio_cable_03, swing1LimitDegrees=5, swing2LimitDegrees=5, twistLimitDegrees=3}
 {bone1Name=dyn_radio_cable_05, bone2Name=dyn_radio_cable_04, swing1LimitDegrees=5, swing2LimitDegrees=5, twistLimitDegrees=3}
 {bone1Name=dyn_radio_cable_06, bone2Name=dyn_radio_cable_05, swing1LimitDegrees=5, swing2LimitDegrees=5, twistLimitDegrees=3}
 {bone1Name=dyn_radio_cable_07, bone2Name=dyn_radio_cable_06, swing1LimitDegrees=5, swing2LimitDegrees=5, twistLimitDegrees=3}
 {bone1Name=dyn_radio_cable_08, bone2Name=dyn_radio_cable_07, swing1LimitDegrees=5, swing2LimitDegrees=5, twistLimitDegrees=3}
]
