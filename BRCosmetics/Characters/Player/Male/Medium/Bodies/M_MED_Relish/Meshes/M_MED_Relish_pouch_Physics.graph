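# Physics asset for the pouches on the M_MED_Relish male character body, giving two smaller pouches and a large back pouch physical secondary motion.
# Each pouch is a box body tightly limited in swing and twist to a tiny capsule master body, which connects to the pelvis with free angular motion.
# Capsules on the pelvis, lower spine and upper thighs form a simplified body volume around the pouches.
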PhysicsAsset /BRCosmetics/Characters/Player/Male/Medium/Bodies/M_MED_Relish/Meshes/M_MED_Relish_pouch_Physics
bodies {
 dyn_pouch_04 [{
   shapeType=Box
   extentX=7.2108
   extentY=4.6604
   extentZ=10.6011
   shapeName=dyn_pouch_04_box
   center=<0.57 0 1.7>
   rotation=<rot 0 92.2733 -81.1672>
  }]
 dyn_pouch_04_master [{shapeType=Capsule, length=0.505, radius=0.505, shapeName=dyn_pouch_04_master_capsule, rotation=<rot 0 -90 0>}]
 dyn_pouch_05 [{
   shapeType=Box
   extentX=7.2604
   extentY=3.9175
   extentZ=10.6011
   shapeName=dyn_pouch_05_box
   center=<1 0 1.743>
   rotation=<rot 0 91.6893 -79.9499>
  }]
 dyn_pouch_05_master [{shapeType=Capsule, length=0.505, radius=0.505, shapeName=dyn_pouch_05_master_capsule, rotation=<rot 0 -90 0>}]
 dyn_pouch_back_large [{
   shapeType=Box
   extentX=14.442
   extentY=8.0331
   extentZ=15.5774
   shapeName=dyn_pouch_back_large_box
   center=<1.0985 0 3.5>
   rotation=<rot 0 0 90>
  }]
 dyn_pouch_back_large_master [{shapeType=Capsule, length=0.505, radius=0.505, shapeName=dyn_pouch_back_large_master_capsule, rotation=<rot 0 -90 0>}]
 pelvis [{shapeType=Capsule, length=10.3973, radius=12.1675, center=<-2.5543 -0.3759 0>}]
 spine_01 [{shapeType=Capsule, length=9.8206, radius=10.1062, shapeName=spine_01_capsule, center=<-0.9292 0 0>}]
 spine_02 [{shapeType=Capsule, length=7.8445, radius=9.2901, center=<-1.892 0 0>}]
 thigh_twist_01_l [{shapeType=Capsule, length=23, radius=9, center=<3.5 0 0>, rotation=<rot 90 0 0>}]
 thigh_twist_01_r [{shapeType=Capsule, length=23, radius=9, center=<-3.5 0 0>, rotation=<rot 90 0 0>}]
}
constraints [
 {bone1Name=spine_02, bone2Name=spine_01, swing1LimitDegrees=45, swing2LimitDegrees=45, twistLimitDegrees=45}
 {bone1Name=thigh_twist_01_l, bone2Name=pelvis, swing1LimitDegrees=45, swing2LimitDegrees=45, twistLimitDegrees=45}
 {bone1Name=thigh_twist_01_r, bone2Name=pelvis, swing1LimitDegrees=45, swing2LimitDegrees=45, twistLimitDegrees=45}
 {
  bone1Name=spine_01
  bone2Name=pelvis
  swing1LimitDegrees=45
  swing1Motion=Free
  swing2LimitDegrees=45
  swing2Motion=Free
  twistLimitDegrees=45
  twistMotion=Free
 }
 {
  bone1Name=dyn_pouch_05_master
  bone2Name=pelvis
  swing1LimitDegrees=45
  swing1Motion=Free
  swing2LimitDegrees=45
  swing2Motion=Free
  twistLimitDegrees=45
  twistMotion=Free
 }
 {bone1Name=dyn_pouch_05, bone2Name=dyn_pouch_05_master, swing1LimitDegrees=10, swing2LimitDegrees=15, twistLimitDegrees=5}
 {bone1Name=dyn_pouch_04, bone2Name=dyn_pouch_04_master, swing1LimitDegrees=10, swing2LimitDegrees=15, twistLimitDegrees=3}
 {
  bone1Name=dyn_pouch_04_master
  bone2Name=pelvis
  swing1LimitDegrees=45
  swing1Motion=Free
  swing2LimitDegrees=45
  swing2Motion=Free
  twistLimitDegrees=45
  twistMotion=Free
 }
 {bone1Name=dyn_pouch_back_large, bone2Name=dyn_pouch_back_large_master, swing1LimitDegrees=5, swing2LimitDegrees=15, twistLimitDegrees=3}
 {
  bone1Name=dyn_pouch_back_large_master
  bone2Name=pelvis
  swing1LimitDegrees=45
  swing1Motion=Free
  swing2LimitDegrees=45
  swing2Motion=Free
  twistLimitDegrees=45
  twistMotion=Free
 }
]
