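# Physics asset for the top mesh of the M_MED_Relish male player body, covering a dynamic radio cable and part of the upper body.
# Capsules cover the pelvis, spine and left arm only, linked in a chain, plus four small cable bodies.
# The cable chains from spine_05 through cable_05, lookat, cable_04 and cable_03, with free rotation at the first two joints and very tight swing and twist limits on the last two.
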PhysicsAsset /BRCosmetics/Characters/Player/Male/Medium/Bodies/M_MED_Relish/Meshes/M_MED_Relish_top_Physics
bodies {
 dyn_radio_top_cable_03 [{
   shapeType=Capsule
   length=1.8367
   radius=0.8886
   shapeName=dyn_radio_top_cable_03_capsule
   center=<-1.5 0 0>
   rotation=<rot 73.9272 179.9996 179.9994>
  }]
 dyn_radio_top_cable_04 [{
   shapeType=Capsule
   length=3.4142
   radius=0.7322
   shapeName=dyn_radio_top_cable_04_capsule
   center=<-2 0 0>
   rotation=<rot 81.2509 179.9999 179.9999>
  }]
 dyn_radio_top_cable_05 [{
   shapeType=Capsule
   length=0.505
   radius=0.505
   shapeName=dyn_radio_top_cable_05_capsule
   center=<-0.0704 0.8245 -0.2543>
   rotation=<rot 0 -0.2135 -75.2769>
  }]
 dyn_radio_top_cable_lookat [{shapeType=Capsule, length=0.505, radius=0.505, shapeName=dyn_radio_top_cable_lookat_capsule, rotation=<rot 0 -90 0>}]
 hand_l [{shapeType=Capsule, length=8.9362, radius=6.6571, center=<7.6413 -2.8158 0.0534>, rotation=<rot 0 89.2872 -66.1886>}]
 lowerarm_l [{shapeType=Capsule, length=14.8982, radius=4.9143, center=<10.3555 0.873 0.8097>, rotation=<rot 0 90.5993 -86.09>}]
 pelvis [{shapeType=Capsule, length=8.8487, radius=12.6667}]
 spine_01 [{shapeType=Capsule, length=11.0332, radius=8.5483, shapeName=spine_01_capsule, center=<0 1.6548 0>}]
 spine_02 [{shapeType=Capsule, length=8.959, radius=9.0551, center=<0 1.3659 0>}]
 spine_03 [{shapeType=Capsule, length=8.4577, radius=9.5005, center=<0 1.8647 0>}]
 spine_04 [{shapeType=Capsule, length=4.3092, radius=13.5051}]
 spine_05 [{shapeType=Capsule, length=12.681, radius=12.9554, center=<-1.7913 1.3149 0>}]
 upperarm_l [{shapeType=Capsule, length=18.6484, radius=6.2659, center=<11.4247 -1.119 -0.1418>, rotation=<rot 0 -96.1904 -85.0371>}]
}
constraints [
 {bone1Name=spine_02, bone2Name=spine_01, swing1LimitDegrees=45, swing2LimitDegrees=45, twistLimitDegrees=45}
 {bone1Name=spine_03, bone2Name=spine_02, swing1LimitDegrees=45, swing2LimitDegrees=45, twistLimitDegrees=45}
 {bone1Name=spine_04, bone2Name=spine_03, swing1LimitDegrees=45, swing2LimitDegrees=45, twistLimitDegrees=45}
 {bone1Name=spine_05, bone2Name=spine_04, swing1LimitDegrees=45, swing2LimitDegrees=45, twistLimitDegrees=45}
 {bone1Name=upperarm_l, bone2Name=spine_05, swing1LimitDegrees=45, swing2LimitDegrees=45, twistLimitDegrees=45}
 {bone1Name=lowerarm_l, bone2Name=upperarm_l, swing1LimitDegrees=45, swing2LimitDegrees=45, twistLimitDegrees=45}
 {bone1Name=hand_l, bone2Name=lowerarm_l, swing1LimitDegrees=45, swing2LimitDegrees=45, twistLimitDegrees=45}
 {
  bone1Name=spine_01
  bone2Name=pelvis
  swing1LimitDegrees=45
  swing1Motion=Free
  swing2LimitDegrees=45
  swing2Motion=Free
  twistLimitDegrees=45
  twistMotion=Free
 }
 {
  bone1Name=dyn_radio_top_cable_lookat
  bone2Name=dyn_radio_top_cable_05
  swing1LimitDegrees=45
  swing1Motion=Free
  swing2LimitDegrees=45
  swing2Motion=Free
  twistLimitDegrees=45
  twistMotion=Free
 }
 {bone1Name=dyn_radio_top_cable_04, bone2Name=dyn_radio_top_cable_lookat, swing1LimitDegrees=3, swing2LimitDegrees=3, twistLimitDegrees=1}
 {bone1Name=dyn_radio_top_cable_03, bone2Name=dyn_radio_top_cable_04, swing1LimitDegrees=3, swing2LimitDegrees=3, twistLimitDegrees=1}
 {
  bone1Name=dyn_radio_top_cable_05
  bone2Name=spine_05
  swing1LimitDegrees=45
  swing1Motion=Free
  swing2LimitDegrees=45
  swing2Motion=Free
  twistLimitDegrees=45
  twistMotion=Free
 }
]
