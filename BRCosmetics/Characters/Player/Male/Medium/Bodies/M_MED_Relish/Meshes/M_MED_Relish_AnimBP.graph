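# Post-copy animation Blueprint for the M_MED_Relish male player body cosmetic, driving body deformation and dangling gear.
# It copies pose and curves from the attached parent mesh, then runs Emperor twist and sculpter Control Rigs plus a Relish rig.
# Long chains of weighted multi-bone Constraint nodes anchor dyn_*_master bones for pouches, radio, zipper, holster, clips and straps to the body.
# RigidBody nodes use a pouch physics asset for secondary motion.
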AnimBlueprint /BRCosmetics/Characters/Player/Male/Medium/Bodies/M_MED_Relish/Meshes/M_MED_Relish_AnimBP
target_skeleton=&/M_MED_Relish_Skeleton
graphs [
 AnimGraph {nodes [
   n1 = "Constraint - dyn_pouch_01_master (PPPPPPPPPPPP)"!(
    ComponentPose
    Alpha=1
    ConstraintWeights_0=0.0043
    ConstraintWeights_1:float=0
    ConstraintWeights_2:float=0
    ConstraintWeights_3:float=0.3289
    ConstraintWeights_4:float=0.0171
    ConstraintWeights_5:float=0.0315
    ConstraintWeights_6:float=0
    ConstraintWeights_7:float=0
    ConstraintWeights_8:float=0
    ConstraintWeights_9:float=0
    ConstraintWeights_10:float=0.6182
    ConstraintWeights_11:float=0
   ) -> (Pose) {
    alphaBoolBlend {blendOption=Linear}
    boneToModify {boneName=dyn_pouch_01_master}
    constraintSetup [
     {targetBone {boneName=spine_03}}
     {targetBone {boneName=clavicle_l}}
     {targetBone {boneName=upperarm_twist_02_l}}
     {targetBone {boneName=spine_01}}
     {targetBone {boneName=spine_02}}
     {targetBone {boneName=thigh_twist_01_l}}
     {targetBone {boneName=spine_04}}
     {targetBone {boneName=spine_05}}
     {targetBone {boneName=upperarm_twist_01_l}}
     {targetBone {boneName=neck_02}}
     {targetBone {boneName=pelvis}}
     {targetBone {boneName=upperarm_l}}
    ]
    constraintWeights [1,1,1,1,1,1,1,1,1,1,1,1]
   }
   n2 = "Constraint - dyn_pouch_02_master (PPPPP)"!(
    ComponentPose=@n1
    Alpha=1
    ConstraintWeights_0=0.0074
    ConstraintWeights_1:float=0.3079
    ConstraintWeights_2:float=0.0169
    ConstraintWeights_3:float=0.0323
    ConstraintWeights_4:float=0.6354
   ) -> (Pose) {
    alphaBoolBlend {blendOption=Linear}
    boneToModify {boneName=dyn_pouch_02_master}
    constraintSetup [
     {targetBone {boneName=spine_03}}
     {targetBone {boneName=spine_01}}
     {targetBone {boneName=spine_02}}
     {targetBone {boneName=thigh_twist_01_l}}
     {targetBone {boneName=pelvis}}
    ]
    constraintWeights [1,1,1,1,1]
   }
   n3 = "Constraint - dyn_pouch_03_master (PPPPP)"!(
    ComponentPose=@n2
    Alpha=1
    ConstraintWeights_0=0.0071
    ConstraintWeights_1:float=0.2904
    ConstraintWeights_2:float=0.0163
    ConstraintWeights_3:float=0.0238
    ConstraintWeights_4:float=0.6623
   ) -> (Pose) {
    alphaBoolBlend {blendOption=Linear}
    boneToModify {boneName=dyn_pouch_03_master}
    constraintSetup [
     {targetBone {boneName=spine_03}}
     {targetBone {boneName=spine_01}}
     {targetBone {boneName=spine_02}}
     {targetBone {boneName=thigh_twist_01_l}}
     {targetBone {boneName=pelvis}}
    ]
    constraintWeights [1,1,1,1,1]
   }
   n4 = "Constraint - dyn_pouch_04_master (PP)"!(ComponentPose=@n3, Alpha=1, ConstraintWeights_0=0.1715, ConstraintWeights_1:float=0.8285) -> (Pose) {
    alphaBoolBlend {blendOption=Linear}
    boneToModify {boneName=dyn_pouch_04_master}
    constraintSetup [{targetBone {boneName=spine_01}}, {targetBone {boneName=pelvis}}]
    constraintWeights [1,1]
   }
   n5 = "Constraint - dyn_pouch_05_master (PPP)"!(ComponentPose=@n4, Alpha=1, ConstraintWeights_0=0.1007, ConstraintWeights_1:float=0.0182, ConstraintWeights_2:float=0.8811) -> (Pose) {
    alphaBoolBlend {blendOption=Linear}
    boneToModify {boneName=dyn_pouch_05_master}
    constraintSetup [{targetBone {boneName=spine_01}}, {targetBone {boneName=thigh_twist_01_r}}, {targetBone {boneName=pelvis}}]
    constraintWeights [1,1,1]
   }
   n6 = "Constraint - dyn_pouch_radio_master (PPPPPP)"!(
    ComponentPose=@n5
    Alpha=1
    ConstraintWeights_0=0
    ConstraintWeights_1:float=0.2202
    ConstraintWeights_2:float=0.008
    ConstraintWeights_3:float=0.0069
    ConstraintWeights_4:float=0.7629
    ConstraintWeights_5:float=0.002
   ) -> (Pose) {
    alphaBoolBlend {blendOption=Linear}
    boneToModify {boneName=dyn_pouch_radio_master}
    constraintSetup [
     {targetBone {boneName=thigh_twist_01_r}}
     {targetBone {boneName=spine_01}}
     {targetBone {boneName=spine_02}}
     {targetBone {boneName=spine_03}}
     {targetBone {boneName=pelvis}}
     {targetBone {boneName=thigh_twist_01_l}}
    ]
    constraintWeights [1,1,1,1,1,1]
   }
   n7 = "Constraint - dyn_radio_base_master (PPPPPPPPPPPPPPPP)"!(
    ComponentPose=@n6
    Alpha=1
    ConstraintWeights_0=0.0002
    ConstraintWeights_1:float=0
    ConstraintWeights_2:float=0.1279
    ConstraintWeights_3:float=0
    ConstraintWeights_4:float=0
    ConstraintWeights_5:float=0
    ConstraintWeights_6:float=0.0016
    ConstraintWeights_7:float=0.0879
    ConstraintWeights_8:float=0.2723
    ConstraintWeights_9:float=0.4734
    ConstraintWeights_10:float=0.0171
    ConstraintWeights_11:float=0.0157
    ConstraintWeights_12:float=0
    ConstraintWeights_13:float=0
    ConstraintWeights_14:float=0.0039
    ConstraintWeights_15:float=0.0001
   ) -> (Pose) {
    alphaBoolBlend {blendOption=Linear}
    boneToModify {boneName=dyn_radio_base_master}
    constraintSetup [
     {targetBone {boneName=head}}
     {targetBone {boneName=upperarm_r}}
     {targetBone {boneName=clavicle_l}}
     {targetBone {boneName=upperarm_twist_02_l}}
     {targetBone {boneName=upperarm_twist_01_r}}
     {targetBone {boneName=spine_01}}
     {targetBone {boneName=spine_02}}
     {targetBone {boneName=spine_03}}
     {targetBone {boneName=spine_04}}
     {targetBone {boneName=spine_05}}
     {targetBone {boneName=upperarm_twist_01_l}}
     {targetBone {boneName=neck_02}}
     {targetBone {boneName=clavicle_r}}
     {targetBone {boneName=pelvis}}
     {targetBone {boneName=neck_01}}
     {targetBone {boneName=upperarm_l}}
    ]
    constraintWeights [1,1,1,1,1,1,1,1,1,1,1,1,1,1,1,1]
   }
   n8 = "Constraint - dyn_zipper_master (PPPPPPPP)"!(
    ComponentPose=@n7
    Alpha=1
    ConstraintWeights_0=0.0002
    ConstraintWeights_1:float=0.0006
    ConstraintWeights_2:float=0.0373
    ConstraintWeights_3:float=0.0994
    ConstraintWeights_4:float=0.7321
    ConstraintWeights_5:float=0.0008
    ConstraintWeights_6:float=0.0857
    ConstraintWeights_7:float=0.0439
   ) -> (Pose) {
    alphaBoolBlend {blendOption=Linear}
    boneToModify {boneName=dyn_zipper_master}
    constraintSetup [
     {targetBone {boneName=head}}
     {targetBone {boneName=clavicle_l}}
     {targetBone {boneName=spine_03}}
     {targetBone {boneName=spine_04}}
     {targetBone {boneName=spine_05}}
     {targetBone {boneName=clavicle_r}}
     {targetBone {boneName=neck_02}}
     {targetBone {boneName=neck_01}}
    ]
    constraintWeights [1,1,1,1,1,1,1,1]
   }
   n9 = "Constraint - dyn_holster_master (PPPPPPPP)"!(
    ComponentPose=@n8
    Alpha=1
    ConstraintWeights_0=0
    ConstraintWeights_1:float=0.0634
    ConstraintWeights_2:float=0.8608
    ConstraintWeights_3:float=0
    ConstraintWeights_4:float=0
    ConstraintWeights_5:float=0
    ConstraintWeights_6:float=0
    ConstraintWeights_7:float=0.0758
   ) -> (Pose) {
    alphaBoolBlend {blendOption=Linear}
    boneToModify {boneName=dyn_holster_master}
    constraintSetup [
     {targetBone {boneName=spine_03}}
     {targetBone {boneName=thigh_r}}
     {targetBone {boneName=thigh_twist_01_r}}
     {targetBone {boneName=spine_01}}
     {targetBone {boneName=spine_02}}
     {targetBone {boneName=thigh_twist_01_l}}
     {targetBone {boneName=thigh_l}}
     {targetBone {boneName=pelvis}}
    ]
    constraintWeights [1,1,1,1,1,1,1,1]
   }
   n10 = "Constraint - dyn_clip_r_shoulder_01_master (PPPPPPPPPPPPPPPP)"!(
    ComponentPose=@n9
    Alpha=1
    ConstraintWeights_0=0.0088
    ConstraintWeights_1:float=0
    ConstraintWeights_2:float=0
    ConstraintWeights_3:float=0
    ConstraintWeights_4:float=0.027
    ConstraintWeights_5:float=0
    ConstraintWeights_6:float=0
    ConstraintWeights_7:float=0.0354
    ConstraintWeights_8:float=0.0529
    ConstraintWeights_9:float=0.4506
    ConstraintWeights_10:float=0
    ConstraintWeights_11:float=0.0885
    ConstraintWeights_12:float=0.3043
    ConstraintWeights_13:float=0
    ConstraintWeights_14:float=0.0325
    ConstraintWeights_15:float=0
   ) -> (Pose) {
    alphaBoolBlend {blendOption=Linear}
    boneToModify {boneName=dyn_clip_r_shoulder_01_master}
    constraintSetup [
     {targetBone {boneName=head}}
     {targetBone {boneName=upperarm_r}}
     {targetBone {boneName=clavicle_l}}
     {targetBone {boneName=upperarm_twist_02_l}}
     {targetBone {boneName=upperarm_twist_01_r}}
     {targetBone {boneName=spine_01}}
     {targetBone {boneName=spine_02}}
     {targetBone {boneName=spine_03}}
     {targetBone {boneName=spine_04}}
     {targetBone {boneName=spine_05}}
     {targetBone {boneName=upperarm_twist_01_l}}
     {targetBone {boneName=neck_02}}
     {targetBone {boneName=clavicle_r}}
     {targetBone {boneName=pelvis}}
     {targetBone {boneName=neck_01}}
     {targetBone {boneName=upperarm_l}}
    ]
    constraintWeights [1,1,1,1,1,1,1,1,1,1,1,1,1,1,1,1]
   }
   n11 = "Constraint - dyn_clip_r_shoulder_02_master (PPPPPPPPPPPPPPPP)"!(
    ComponentPose=@n10
    Alpha=1
    ConstraintWeights_0=0.0056
    ConstraintWeights_1:float=0
    ConstraintWeights_2:float=0
    ConstraintWeights_3:float=0
    ConstraintWeights_4:float=0.0202
    ConstraintWeights_5:float=0
    ConstraintWeights_6:float=0
    ConstraintWeights_7:float=0.0424
    ConstraintWeights_8:float=0.0661
    ConstraintWeights_9:float=0.4708
    ConstraintWeights_10:float=0
    ConstraintWeights_11:float=0.0749
    ConstraintWeights_12:float=0.2932
    ConstraintWeights_13:float=0
    ConstraintWeights_14:float=0.0268
    ConstraintWeights_15:float=0
   ) -> (Pose) {
    alphaBoolBlend {blendOption=Linear}
    boneToModify {boneName=dyn_clip_r_shoulder_02_master}
    constraintSetup [
     {targetBone {boneName=head}}
     {targetBone {boneName=upperarm_r}}
     {targetBone {boneName=clavicle_l}}
     {targetBone {boneName=upperarm_twist_02_l}}
     {targetBone {boneName=upperarm_twist_01_r}}
     {targetBone {boneName=spine_01}}
     {targetBone {boneName=spine_02}}
     {targetBone {boneName=spine_03}}
     {targetBone {boneName=spine_04}}
     {targetBone {boneName=spine_05}}
     {targetBone {boneName=upperarm_twist_01_l}}
     {targetBone {boneName=neck_02}}
     {targetBone {boneName=clavicle_r}}
     {targetBone {boneName=pelvis}}
     {targetBone {boneName=neck_01}}
     {targetBone {boneName=upperarm_l}}
    ]
    constraintWeights [1,1,1,1,1,1,1,1,1,1,1,1,1,1,1,1]
   }
   n12 = "Constraint - dyn_clip_r_shoulder_01_master (PP)"!(ComponentPose=@n11, Alpha=1, ConstraintWeights_0=0.8, ConstraintWeights_1:float=0.2) -> (Pose) {
    alphaBoolBlend {blendOption=Linear}
    boneToModify {boneName=dyn_clip_r_shoulder_01_master}
    constraintSetup [{targetBone {boneName=spine_05}}, {targetBone {boneName=clavicle_l}}]
    constraintWeights [1,1]
   }
   n13 = "Constraint - dyn_clip_r_shoulder_02_master (PPPPPPPPPPPPPPPP)"(
    ComponentPose=@n12
    Alpha=1
    ConstraintWeights_0=0.0001
    ConstraintWeights_1=0
    ConstraintWeights_2=0.1949
    ConstraintWeights_3=0
    ConstraintWeights_4=0
    ConstraintWeights_5=0
    ConstraintWeights_6=0
    ConstraintWeights_7=0.002
    ConstraintWeights_8=0.0053
    ConstraintWeights_9=0.7929
    ConstraintWeights_10=0.0006
    ConstraintWeights_11=0.003
    ConstraintWeights_12=0
    ConstraintWeights_13=0
    ConstraintWeights_14=0.0013
    ConstraintWeights_15=0
   ) {
    alphaBoolBlend {blendOption=Linear}
    boneToModify {boneName=dyn_clip_r_shoulder_02_master}
    constraintSetup [
     {targetBone {boneName=head}}
     {targetBone {boneName=upperarm_r}}
     {targetBone {boneName=clavicle_l}}
     {targetBone {boneName=upperarm_twist_02_l}}
     {targetBone {boneName=upperarm_twist_01_r}}
     {targetBone {boneName=spine_01}}
     {targetBone {boneName=spine_02}}
     {targetBone {boneName=spine_03}}
     {targetBone {boneName=spine_04}}
     {targetBone {boneName=spine_05}}
     {targetBone {boneName=upperarm_twist_01_l}}
     {targetBone {boneName=neck_02}}
     {targetBone {boneName=clavicle_r}}
     {targetBone {boneName=pelvis}}
     {targetBone {boneName=neck_01}}
     {targetBone {boneName=upperarm_l}}
    ]
    constraintWeights [1,1,1,1,1,1,1,1,1,1,1,1,1,1,1,1]
   }
   n14 = "Constraint - dyn_clip_shoulder_l_back_01_master (PPPPPPPPPPPPPPPP)"!(
    ComponentPose=@n13
    Alpha=1
    ConstraintWeights_0=0.0016
    ConstraintWeights_1:float=0
    ConstraintWeights_2:float=0.3245
    ConstraintWeights_3:float=0.0001
    ConstraintWeights_4:float=0
    ConstraintWeights_5:float=0
    ConstraintWeights_6:float=0
    ConstraintWeights_7:float=0.0906
    ConstraintWeights_8:float=0.0512
    ConstraintWeights_9:float=0.4513
    ConstraintWeights_10:float=0.0163
    ConstraintWeights_11:float=0.0479
    ConstraintWeights_12:float=0.0001
    ConstraintWeights_13:float=0
    ConstraintWeights_14:float=0.0163
    ConstraintWeights_15:float=0.0002
   ) -> (Pose) {
    alphaBoolBlend {blendOption=Linear}
    boneToModify {boneName=dyn_clip_shoulder_l_back_01_master}
    constraintSetup [
     {targetBone {boneName=head}}
     {targetBone {boneName=upperarm_r}}
     {targetBone {boneName=clavicle_l}}
     {targetBone {boneName=upperarm_twist_02_l}}
     {targetBone {boneName=upperarm_twist_01_r}}
     {targetBone {boneName=spine_01}}
     {targetBone {boneName=spine_02}}
     {targetBone {boneName=spine_03}}
     {targetBone {boneName=spine_04}}
     {targetBone {boneName=spine_05}}
     {targetBone {boneName=upperarm_twist_01_l}}
     {targetBone {boneName=neck_02}}
     {targetBone {boneName=clavicle_r}}
     {targetBone {boneName=pelvis}}
     {targetBone {boneName=neck_01}}
     {targetBone {boneName=upperarm_l}}
    ]
    constraintWeights [1,1,1,1,1,1,1,1,1,1,1,1,1,1,1,1]
   }
   n15 = "Constraint - dyn_clip_shoulder_l_back_02_master (PPPPPPPPPPPPPPPPP)"!(
    ComponentPose=@n14
    Alpha=1
    ConstraintWeights_0=0
    ConstraintWeights_1:float=0
    ConstraintWeights_2:float=0.101
    ConstraintWeights_3:float=0
    ConstraintWeights_4:float=0
    ConstraintWeights_5:float=0.0001
    ConstraintWeights_6:float=0
    ConstraintWeights_7:float=0
    ConstraintWeights_8:float=0.1664
    ConstraintWeights_9:float=0.2192
    ConstraintWeights_10:float=0.5101
    ConstraintWeights_11:float=0.0013
    ConstraintWeights_12:float=0.0013
    ConstraintWeights_13:float=0
    ConstraintWeights_14:float=0
    ConstraintWeights_15:float=0.0004
    ConstraintWeights_16:float=0
   ) -> (Pose) {
    alphaBoolBlend {blendOption=Linear}
    boneToModify {boneName=dyn_clip_shoulder_l_back_02_master}
    constraintSetup [
     {targetBone {boneName=head}}
     {targetBone {boneName=upperarm_r}}
     {targetBone {boneName=clavicle_l}}
     {targetBone {boneName=upperarm_twist_02_l}}
     {targetBone {boneName=upperarm_twist_01_r}}
     {targetBone {boneName=deform_deltoid_fr_l}}
     {targetBone {boneName=spine_01}}
     {targetBone {boneName=spine_02}}
     {targetBone {boneName=spine_03}}
     {targetBone {boneName=spine_04}}
     {targetBone {boneName=spine_05}}
     {targetBone {boneName=upperarm_twist_01_l}}
     {targetBone {boneName=neck_02}}
     {targetBone {boneName=clavicle_r}}
     {targetBone {boneName=pelvis}}
     {targetBone {boneName=neck_01}}
     {targetBone {boneName=upperarm_l}}
    ]
    constraintWeights [1,1,1,1,1,1,1,1,1,1,1,1,1,1,1,1,1]
   }
   n16 = "Constraint - dyn_clip_shoulder_r_back_01_master (PPPPPPPPPPPPPPPP)"!(
    ComponentPose=@n15
    Alpha=1
    ConstraintWeights_0=0.0008
    ConstraintWeights_1:float=0
    ConstraintWeights_2:float=0
    ConstraintWeights_3:float=0
    ConstraintWeights_4:float=0.0054
    ConstraintWeights_5:float=0
    ConstraintWeights_6:float=0
    ConstraintWeights_7:float=0.0966
    ConstraintWeights_8:float=0.0375
    ConstraintWeights_9:float=0.4405
    ConstraintWeights_10:float=0
    ConstraintWeights_11:float=0.0452
    ConstraintWeights_12:float=0.3702
    ConstraintWeights_13:float=0
    ConstraintWeights_14:float=0.0038
    ConstraintWeights_15:float=0
   ) -> (Pose) {
    alphaBoolBlend {blendOption=Linear}
    boneToModify {boneName=dyn_clip_shoulder_r_back_01_master}
    constraintSetup [
     {targetBone {boneName=head}}
     {targetBone {boneName=upperarm_r}}
     {targetBone {boneName=clavicle_l}}
     {targetBone {boneName=upperarm_twist_02_l}}
     {targetBone {boneName=upperarm_twist_01_r}}
     {targetBone {boneName=spine_01}}
     {targetBone {boneName=spine_02}}
     {targetBone {boneName=spine_03}}
     {targetBone {boneName=spine_04}}
     {targetBone {boneName=spine_05}}
     {targetBone {boneName=upperarm_twist_01_l}}
     {targetBone {boneName=neck_02}}
     {targetBone {boneName=clavicle_r}}
     {targetBone {boneName=pelvis}}
     {targetBone {boneName=neck_01}}
     {targetBone {boneName=upperarm_l}}
    ]
    constraintWeights [1,1,1,1,1,1,1,1,1,1,1,1,1,1,1,1]
   }
   n17 = "Constraint - dyn_clip_shoulder_r_back_02_master (PPPPPPPPPPPPPPPP)"!(
    ComponentPose=@n16
    Alpha=1
    ConstraintWeights_0=0
    ConstraintWeights_1:float=0
    ConstraintWeights_2:float=0
    ConstraintWeights_3:float=0
    ConstraintWeights_4:float=0.0016
    ConstraintWeights_5:float=0
    ConstraintWeights_6:float=0.0002
    ConstraintWeights_7:float=0.1701
    ConstraintWeights_8:float=0.219
    ConstraintWeights_9:float=0.4668
    ConstraintWeights_10:float=0
    ConstraintWeights_11:float=0.0015
    ConstraintWeights_12:float=0.1406
    ConstraintWeights_13:float=0
    ConstraintWeights_14:float=0.0001
    ConstraintWeights_15:float=0
   ) -> (Pose) {
    alphaBoolBlend {blendOption=Linear}
    boneToModify {boneName=dyn_clip_shoulder_r_back_02_master}
    constraintSetup [
     {targetBone {boneName=head}}
     {targetBone {boneName=upperarm_r}}
     {targetBone {boneName=clavicle_l}}
     {targetBone {boneName=upperarm_twist_02_l}}
     {targetBone {boneName=upperarm_twist_01_r}}
     {targetBone {boneName=spine_01}}
     {targetBone {boneName=spine_02}}
     {targetBone {boneName=spine_03}}
     {targetBone {boneName=spine_04}}
     {targetBone {boneName=spine_05}}
     {targetBone {boneName=upperarm_twist_01_l}}
     {targetBone {boneName=neck_02}}
     {targetBone {boneName=clavicle_r}}
     {targetBone {boneName=pelvis}}
     {targetBone {boneName=neck_01}}
     {targetBone {boneName=upperarm_l}}
    ]
    constraintWeights [1,1,1,1,1,1,1,1,1,1,1,1,1,1,1,1]
   }
   n18 = "Constraint - dyn_clip_back_l_01_master (PPPPPPPPPPP)"!(
    ComponentPose=@n17
    Alpha=1
    ConstraintWeights_0=0.0001
    ConstraintWeights_1:float=0
    ConstraintWeights_2:float=0.1291
    ConstraintWeights_3:float=0.2494
    ConstraintWeights_4:float=0.3087
    ConstraintWeights_5:float=0.1993
    ConstraintWeights_6:float=0.0439
    ConstraintWeights_7:float=0
    ConstraintWeights_8:float=0
    ConstraintWeights_9:float=0.0695
    ConstraintWeights_10:float=0
   ) -> (Pose) {
    alphaBoolBlend {blendOption=Linear}
    boneToModify {boneName=dyn_clip_back_l_01_master}
    constraintSetup [
     {targetBone {boneName=clavicle_l}}
     {targetBone {boneName=upperarm_twist_02_l}}
     {targetBone {boneName=spine_01}}
     {targetBone {boneName=spine_02}}
     {targetBone {boneName=spine_03}}
     {targetBone {boneName=spine_04}}
     {targetBone {boneName=spine_05}}
     {targetBone {boneName=upperarm_twist_01_l}}
     {targetBone {boneName=neck_02}}
     {targetBone {boneName=pelvis}}
     {targetBone {boneName=upperarm_l}}
    ]
    constraintWeights [1,1,1,1,1,1,1,1,1,1,1]
   }
   n19 = "Constraint - dyn_clip_back_l_02_master (PPPPPPPPPPP)"!(
    ComponentPose=@n18
    Alpha=1
    ConstraintWeights_0=0
    ConstraintWeights_1:float=0
    ConstraintWeights_2:float=0.2628
    ConstraintWeights_3:float=0.2538
    ConstraintWeights_4:float=0.1992
    ConstraintWeights_5:float=0.0764
    ConstraintWeights_6:float=0.0118
    ConstraintWeights_7:float=0
    ConstraintWeights_8:float=0
    ConstraintWeights_9:float=0.1961
    ConstraintWeights_10:float=0
   ) -> (Pose) {
    alphaBoolBlend {blendOption=Linear}
    boneToModify {boneName=dyn_clip_back_l_02_master}
    constraintSetup [
     {targetBone {boneName=clavicle_l}}
     {targetBone {boneName=upperarm_twist_02_l}}
     {targetBone {boneName=spine_01}}
     {targetBone {boneName=spine_02}}
     {targetBone {boneName=spine_03}}
     {targetBone {boneName=spine_04}}
     {targetBone {boneName=spine_05}}
     {targetBone {boneName=upperarm_twist_01_l}}
     {targetBone {boneName=neck_02}}
     {targetBone {boneName=pelvis}}
     {targetBone {boneName=upperarm_l}}
    ]
    constraintWeights [1,1,1,1,1,1,1,1,1,1,1]
   }
   n20 = "Constraint - dyn_clip_back_r_01_master (PPPPPPPPPPP)"!(
    ComponentPose=@n19
    Alpha=1
    ConstraintWeights_0=0
    ConstraintWeights_1:float=0
    ConstraintWeights_2:float=0.1337
    ConstraintWeights_3:float=0.251
    ConstraintWeights_4:float=0.3051
    ConstraintWeights_5:float=0.195
    ConstraintWeights_6:float=0.0432
    ConstraintWeights_7:float=0
    ConstraintWeights_8:float=0
    ConstraintWeights_9:float=0.0719
    ConstraintWeights_10:float=0
   ) -> (Pose) {
    alphaBoolBlend {blendOption=Linear}
    boneToModify {boneName=dyn_clip_back_r_01_master}
    constraintSetup [
     {targetBone {boneName=clavicle_l}}
     {targetBone {boneName=upperarm_twist_02_l}}
     {targetBone {boneName=spine_01}}
     {targetBone {boneName=spine_02}}
     {targetBone {boneName=spine_03}}
     {targetBone {boneName=spine_04}}
     {targetBone {boneName=spine_05}}
     {targetBone {boneName=upperarm_twist_01_l}}
     {targetBone {boneName=neck_02}}
     {targetBone {boneName=pelvis}}
     {targetBone {boneName=upperarm_l}}
    ]
    constraintWeights [1,1,1,1,1,1,1,1,1,1,1]
   }
   n21 = "Constraint - dyn_clip_back_r_02_master (PPPPPPPPPPP)"!(
    ComponentPose=@n20
    Alpha=1
    ConstraintWeights_0=0
    ConstraintWeights_1:float=0
    ConstraintWeights_2:float=0.2617
    ConstraintWeights_3:float=0.2493
    ConstraintWeights_4:float=0.1985
    ConstraintWeights_5:float=0.0799
    ConstraintWeights_6:float=0.0132
    ConstraintWeights_7:float=0
    ConstraintWeights_8:float=0
    ConstraintWeights_9:float=0.1975
    ConstraintWeights_10:float=0
   ) -> (Pose) {
    alphaBoolBlend {blendOption=Linear}
    boneToModify {boneName=dyn_clip_back_r_02_master}
    constraintSetup [
     {targetBone {boneName=clavicle_l}}
     {targetBone {boneName=upperarm_twist_02_l}}
     {targetBone {boneName=spine_01}}
     {targetBone {boneName=spine_02}}
     {targetBone {boneName=spine_03}}
     {targetBone {boneName=spine_04}}
     {targetBone {boneName=spine_05}}
     {targetBone {boneName=upperarm_twist_01_l}}
     {targetBone {boneName=neck_02}}
     {targetBone {boneName=pelvis}}
     {targetBone {boneName=upperarm_l}}
    ]
    constraintWeights [1,1,1,1,1,1,1,1,1,1,1]
   }
   n22 = "Constraint - dyn_underarm_clip_r_master (PPPPPPPPPPPPPPP)"!(
    ComponentPose=@n21
    Alpha=1
    ConstraintWeights_0=0.2723
    ConstraintWeights_1:float=0
    ConstraintWeights_2:float=0
    ConstraintWeights_3:float=0
    ConstraintWeights_4:float=0
    ConstraintWeights_5:float=0
    ConstraintWeights_6:float=0.2121
    ConstraintWeights_7:float=0.3171
    ConstraintWeights_8:float=0
    ConstraintWeights_9:float=0.1014
    ConstraintWeights_10:float=0
    ConstraintWeights_11:float=0
    ConstraintWeights_12:float=0
    ConstraintWeights_13:float=0.097
    ConstraintWeights_14:float=0
   ) -> (Pose) {
    alphaBoolBlend {blendOption=Linear}
    boneToModify {boneName=dyn_underarm_clip_r_master}
    constraintSetup [
     {targetBone {boneName=spine_03}}
     {targetBone {boneName=thigh_l}}
     {targetBone {boneName=clavicle_l}}
     {targetBone {boneName=upperarm_twist_02_l}}
     {targetBone {boneName=thigh_r}}
     {targetBone {boneName=thigh_twist_01_r}}
     {targetBone {boneName=spine_01}}
     {targetBone {boneName=spine_02}}
     {targetBone {boneName=thigh_twist_01_l}}
     {targetBone {boneName=spine_04}}
     {targetBone {boneName=spine_05}}
     {targetBone {boneName=upperarm_twist_01_l}}
     {targetBone {boneName=neck_02}}
     {targetBone {boneName=pelvis}}
     {targetBone {boneName=upperarm_l}}
    ]
    constraintWeights [1,1,1,1,1,1,1,1,1,1,1,1,1,1,1]
   }
   n23 = "Constraint - dyn_underarm_clip_l_master (PPPPPPPPPPP)"!(
    ComponentPose=@n22
    Alpha=1
    ConstraintWeights_0=0
    ConstraintWeights_1:float=0
    ConstraintWeights_2:float=0.215
    ConstraintWeights_3:float=0.3231
    ConstraintWeights_4:float=0.272
    ConstraintWeights_5:float=0.0967
    ConstraintWeights_6:float=0
    ConstraintWeights_7:float=0
    ConstraintWeights_8:float=0
    ConstraintWeights_9:float=0.0932
    ConstraintWeights_10:float=0
   ) -> (Pose) {
    alphaBoolBlend {blendOption=Linear}
    boneToModify {boneName=dyn_underarm_clip_l_master}
    constraintSetup [
     {targetBone {boneName=clavicle_l}}
     {targetBone {boneName=upperarm_twist_02_l}}
     {targetBone {boneName=spine_01}}
     {targetBone {boneName=spine_02}}
     {targetBone {boneName=spine_03}}
     {targetBone {boneName=spine_04}}
     {targetBone {boneName=spine_05}}
     {targetBone {boneName=upperarm_twist_01_l}}
     {targetBone {boneName=neck_02}}
     {targetBone {boneName=pelvis}}
     {targetBone {boneName=upperarm_l}}
    ]
    constraintWeights [1,1,1,1,1,1,1,1,1,1,1]
   }
   n24 = "Constraint - dyn_underarm_strap_r_master (PPPPPPP)"!(
    ComponentPose=@n23
    Alpha=1
    ConstraintWeights_0=0.0904
    ConstraintWeights_1:float=0.3063
    ConstraintWeights_2:float=0.3759
    ConstraintWeights_3:float=0.2104
    ConstraintWeights_4:float=0.0002
    ConstraintWeights_5:float=0
    ConstraintWeights_6:float=0.0168
   ) -> (Pose) {
    alphaBoolBlend {blendOption=Linear}
    boneToModify {boneName=dyn_underarm_strap_r_master}
    constraintSetup [
     {targetBone {boneName=spine_01}}
     {targetBone {boneName=spine_02}}
     {targetBone {boneName=spine_03}}
     {targetBone {boneName=spine_04}}
     {targetBone {boneName=spine_05}}
     {targetBone {boneName=clavicle_r}}
     {targetBone {boneName=pelvis}}
    ]
    constraintWeights [1,1,1,1,1,1,1]
   }
   n25 = "Constraint - dyn_underarm_strap_l_master (PPPPP)"!(
    ComponentPose=@n24
    Alpha=1
    ConstraintWeights_0=0.0866
    ConstraintWeights_1:float=0.3116
    ConstraintWeights_2:float=0.3789
    ConstraintWeights_3:float=0.2078
    ConstraintWeights_4:float=0.0152
   ) -> (Pose) {
    alphaBoolBlend {blendOption=Linear}
    boneToModify {boneName=dyn_underarm_strap_l_master}
    constraintSetup [
     {targetBone {boneName=spine_01}}
     {targetBone {boneName=spine_02}}
     {targetBone {boneName=spine_03}}
     {targetBone {boneName=spine_04}}
     {targetBone {boneName=pelvis}}
    ]
    constraintWeights [1,1,1,1,1]
   }
   n26 = "Constraint - dyn_clip_r_shoulder_01_master (PPPPPPPPPPPPPPPP)"(
    ComponentPose=@n25
    Alpha=1
    ConstraintWeights_0=0.0088
    ConstraintWeights_1=0
    ConstraintWeights_2=0
    ConstraintWeights_3=0
    ConstraintWeights_4=0.0274
    ConstraintWeights_5=0
    ConstraintWeights_6=0
    ConstraintWeights_7=0.0354
    ConstraintWeights_8=0.0529
    ConstraintWeights_9=0.4497
    ConstraintWeights_10=0
    ConstraintWeights_11=0.0882
    ConstraintWeights_12=0.3052
    ConstraintWeights_13=0
    ConstraintWeights_14=0.0324
    ConstraintWeights_15=0
   ) {
    alphaBoolBlend {blendOption=Linear}
    boneToModify {boneName=dyn_clip_r_shoulder_01_master}
    constraintSetup [
     {targetBone {boneName=head}}
     {targetBone {boneName=upperarm_r}}
     {targetBone {boneName=clavicle_l}}
     {targetBone {boneName=upperarm_twist_02_l}}
     {targetBone {boneName=upperarm_twist_01_r}}
     {targetBone {boneName=spine_01}}
     {targetBone {boneName=spine_02}}
     {targetBone {boneName=spine_03}}
     {targetBone {boneName=spine_04}}
     {targetBone {boneName=spine_05}}
     {targetBone {boneName=upperarm_twist_01_l}}
     {targetBone {boneName=neck_02}}
     {targetBone {boneName=clavicle_r}}
     {targetBone {boneName=pelvis}}
     {targetBone {boneName=neck_01}}
     {targetBone {boneName=upperarm_l}}
    ]
    constraintWeights [1,1,1,1,1,1,1,1,1,1,1,1,1,1,1,1]
   }
   n27 = "Constraint - dyn_clip_r_shoulder_02_master (PPPPPPPPPPPPPPPP)"(
    ComponentPose=@n26
    Alpha=1
    ConstraintWeights_0=0.0043
    ConstraintWeights_1=0
    ConstraintWeights_2=0
    ConstraintWeights_3=0
    ConstraintWeights_4=0.0159
    ConstraintWeights_5=0
    ConstraintWeights_6=0
    ConstraintWeights_7=0.0466
    ConstraintWeights_8=0.0714
    ConstraintWeights_9=0.4755
    ConstraintWeights_10=0
    ConstraintWeights_11=0.0674
    ConstraintWeights_12=0.2963
    ConstraintWeights_13=0
    ConstraintWeights_14=0.0226
    ConstraintWeights_15=0
   ) {
    alphaBoolBlend {blendOption=Linear}
    boneToModify {boneName=dyn_clip_r_shoulder_02_master}
    constraintSetup [
     {targetBone {boneName=head}}
     {targetBone {boneName=upperarm_r}}
     {targetBone {boneName=clavicle_l}}
     {targetBone {boneName=upperarm_twist_02_l}}
     {targetBone {boneName=upperarm_twist_01_r}}
     {targetBone {boneName=spine_01}}
     {targetBone {boneName=spine_02}}
     {targetBone {boneName=spine_03}}
     {targetBone {boneName=spine_04}}
     {targetBone {boneName=spine_05}}
     {targetBone {boneName=upperarm_twist_01_l}}
     {targetBone {boneName=neck_02}}
     {targetBone {boneName=clavicle_r}}
     {targetBone {boneName=pelvis}}
     {targetBone {boneName=neck_01}}
     {targetBone {boneName=upperarm_l}}
    ]
    constraintWeights [1,1,1,1,1,1,1,1,1,1,1,1,1,1,1,1]
   }
   n28 = "Constraint - dyn_clip_l_shoulder_01_master (PPPPPPPPPPPPPPPP)"!(
    ComponentPose=@n27
    Alpha=1
    ConstraintWeights_0=0
    ConstraintWeights_1:float=0
    ConstraintWeights_2:float=0.2
    ConstraintWeights_3:float=0
    ConstraintWeights_4:float=0
    ConstraintWeights_5:float=0
    ConstraintWeights_6:float=0
    ConstraintWeights_7:float=0
    ConstraintWeights_8:float=0
    ConstraintWeights_9:float=0.8
    ConstraintWeights_10:float=0
    ConstraintWeights_11:float=0
    ConstraintWeights_12:float=0
    ConstraintWeights_13:float=0
    ConstraintWeights_14:float=0
    ConstraintWeights_15:float=0
   ) -> (Pose) {
    alphaBoolBlend {blendOption=Linear}
    boneToModify {boneName=dyn_clip_l_shoulder_01_master}
    constraintSetup [
     {targetBone {boneName=head}}
     {targetBone {boneName=upperarm_r}}
     {targetBone {boneName=clavicle_l}}
     {targetBone {boneName=upperarm_twist_02_l}}
     {targetBone {boneName=upperarm_twist_01_r}}
     {targetBone {boneName=spine_01}}
     {targetBone {boneName=spine_02}}
     {targetBone {boneName=spine_03}}
     {targetBone {boneName=spine_04}}
     {targetBone {boneName=spine_05}}
     {targetBone {boneName=upperarm_twist_01_l}}
     {targetBone {boneName=neck_02}}
     {targetBone {boneName=clavicle_r}}
     {targetBone {boneName=pelvis}}
     {targetBone {boneName=neck_01}}
     {targetBone {boneName=upperarm_l}}
    ]
    constraintWeights [1,1,1,1,1,1,1,1,1,1,1,1,1,1,1,1]
   }
   n29 = "Constraint - dyn_clip_l_shoulder_02_master (PPPPPPPPPPPPPPPP)"!(
    ComponentPose=@n28
    Alpha=1
    ConstraintWeights_0=0.0001
    ConstraintWeights_1:float=0
    ConstraintWeights_2:float=0.1972
    ConstraintWeights_3:float=0
    ConstraintWeights_4:float=0
    ConstraintWeights_5:float=0
    ConstraintWeights_6:float=0
    ConstraintWeights_7:float=0.0011
    ConstraintWeights_8:float=0.0029
    ConstraintWeights_9:float=0.7962
    ConstraintWeights_10:float=0.0003
    ConstraintWeights_11:float=0.0016
    ConstraintWeights_12:float=0
    ConstraintWeights_13:float=0
    ConstraintWeights_14:float=0.0007
    ConstraintWeights_15:float=0
   ) -> (Pose) {
    alphaBoolBlend {blendOption=Linear}
    boneToModify {boneName=dyn_clip_l_shoulder_02_master}
    constraintSetup [
     {targetBone {boneName=head}}
     {targetBone {boneName=upperarm_r}}
     {targetBone {boneName=clavicle_l}}
     {targetBone {boneName=upperarm_twist_02_l}}
     {targetBone {boneName=upperarm_twist_01_r}}
     {targetBone {boneName=spine_01}}
     {targetBone {boneName=spine_02}}
     {targetBone {boneName=spine_03}}
     {targetBone {boneName=spine_04}}
     {targetBone {boneName=spine_05}}
     {targetBone {boneName=upperarm_twist_01_l}}
     {targetBone {boneName=neck_02}}
     {targetBone {boneName=clavicle_r}}
     {targetBone {boneName=pelvis}}
     {targetBone {boneName=neck_01}}
     {targetBone {boneName=upperarm_l}}
    ]
    constraintWeights [1,1,1,1,1,1,1,1,1,1,1,1,1,1,1,1]
   }
   Component To Local(ComponentPose=@n29)
   RigidBody() {
    bOverrideWorldGravity=true
    overridePhysicsAsset=&/M_MED_Relish_pouch_Physics
    simulationSpace=BaseBoneSpace
    alphaBoolBlend {blendOption=Linear}
    baseBoneRef {boneName=pelvis}
    componentAppliedLinearAccClamp=<3000 3000 3000>
    componentLinearAccScale=<0.2 0.2 0.4>
    componentLinearVelScale=<0.1 0.1 0.2>
    overrideWorldGravity {z=-980}
   }
   RigidBody!(ComponentPose, Alpha, OverrideWorldGravity:Vector="0,0,0", ExternalForce) -> (Pose) {
    bClampLinearTranslationLimitToRefPose=true
    bOverrideWorldGravity=true
    overridePhysicsAsset=&M_MED_Relish_pouch_Physics
    simulationSpace=BaseBoneSpace
    alphaBoolBlend {blendOption=Linear}
    baseBoneRef {boneName=root}
    componentAppliedLinearAccClamp=<3000 3000 3000>
    componentLinearAccScale=<0.25 0.25 0.4>
    componentLinearVelScale=<0.2 0.2 0.2>
   }
   n30 = Get bIsCrouching()
   n31 = Control Rig(Source=Copy Pose From Mesh() {bCopyCurves=true, bUseAttachedParent=true}) {
    alphaBoolBlend {blendOption=Linear}
    controlRigAssetReference {blueprintRigClass=/Game/Characters/Player/Male/Medium/Bodies/M_MED_Emperor/Meshes/M_MED_Emperor_Deform_CtrlRig_twists.M_MED_Emperor_Deform_CtrlRig_twists_C}
   }
   n32 = Control Rig!(
    Source=@n31
    Alpha
    KneeRot_fr_l_114deg:Vector
    KneeTrans_fr_l_114deg:Vector="0.258,-3.591,0.009"
    KneeScale_fr_l_114deg:Vector="1,1,1"
    KneeScale_fr_l_130deg:Vector="1,1,1"
    KneeRot_fr_l_130deg:Vector="0,0,-51.2"
    KneeTrans_fr_l_130deg:Vector="0.258,-3.591,0.009"
   ) -> (Pose) {
    alphaBoolBlend {blendOption=Linear}
    controlRigAssetReference {blueprintRigClass=/Game/Characters/Player/Male/Medium/Bodies/M_MED_Emperor/Meshes/M_MED_Emperor_Deform_CtrlRig_sculpters.M_MED_Emperor_Deform_CtrlRig_sculpters_C}
   }
   n33 = Control Rig!(Source=@n32, Alpha) -> (Pose) {
    alphaBoolBlend {blendOption=Linear}
    controlRigAssetReference {blueprintRigClass=&/M_MED_Relish_CtrlRig.M_MED_Relish_CtrlRig_C}
   }
   n34 = "Constraint - dyn_clip_shoulder_l_back_01_master (PPPPPPPP)"!(
    ComponentPose=Local To Component(LocalPose=@n33)
    Alpha=1
    ConstraintWeights_0=0.4346
    ConstraintWeights_1:float=0.0087
    ConstraintWeights_2:float=0.0735
    ConstraintWeights_3:float=0.3587
    ConstraintWeights_4:float=0.0106
    ConstraintWeights_5:float=0.0145
    ConstraintWeights_6:float=0.0552
    ConstraintWeights_7:float=0.0442
   ) -> (Pose) {
    alphaBoolBlend {blendOption=Linear}
    boneToModify {boneName=dyn_clip_shoulder_l_back_01_master}
    constraintSetup [
     {targetBone {boneName=clavicle_l}}
     {targetBone {boneName=spine_03}}
     {targetBone {boneName=spine_04}}
     {targetBone {boneName=spine_05}}
     {targetBone {boneName=upperarm_twist_01_l}}
     {targetBone {boneName=clavicle_r}}
     {targetBone {boneName=neck_02}}
     {targetBone {boneName=neck_01}}
    ]
    constraintWeights [1,1,1,1,1,1,1,1]
   }
   n35 = "Constraint - dyn_clip_shoulder_l_back_02_master (PPPPPPPP)"!(
    ComponentPose=@n34
    Alpha=1
    ConstraintWeights_0=0.2667
    ConstraintWeights_1:float=0.0327
    ConstraintWeights_2:float=0.0573
    ConstraintWeights_3:float=0.202
    ConstraintWeights_4:float=0.4062
    ConstraintWeights_5:float=0.019
    ConstraintWeights_6:float=0.0001
    ConstraintWeights_7:float=0.0159
   ) -> (Pose) {
    alphaBoolBlend {blendOption=Linear}
    boneToModify {boneName=dyn_clip_shoulder_l_back_02_master}
    constraintSetup [
     {targetBone {boneName=clavicle_l}}
     {targetBone {boneName=spine_02}}
     {targetBone {boneName=spine_03}}
     {targetBone {boneName=spine_04}}
     {targetBone {boneName=spine_05}}
     {targetBone {boneName=upperarm_twist_01_l}}
     {targetBone {boneName=neck_02}}
     {targetBone {boneName=upperarm_l}}
    ]
    constraintWeights [1,1,1,1,1,1,1,1]
   }
   n36 = "Constraint - dyn_clip_shoulder_r_back_01_master (PPPPPPPP)"!(
    ComponentPose=@n35
    Alpha=1
    ConstraintWeights_0=0.0335
    ConstraintWeights_1:float=0.0113
    ConstraintWeights_2:float=0.0347
    ConstraintWeights_3:float=0.0763
    ConstraintWeights_4:float=0.3685
    ConstraintWeights_5:float=0.4135
    ConstraintWeights_6:float=0.0181
    ConstraintWeights_7:float=0.0441
   ) -> (Pose) {
    alphaBoolBlend {blendOption=Linear}
    boneToModify {boneName=dyn_clip_shoulder_r_back_01_master}
    constraintSetup [
     {targetBone {boneName=clavicle_l}}
     {targetBone {boneName=upperarm_twist_01_r}}
     {targetBone {boneName=deform_deltoid_fr_r}}
     {targetBone {boneName=spine_04}}
     {targetBone {boneName=spine_05}}
     {targetBone {boneName=clavicle_r}}
     {targetBone {boneName=neck_02}}
     {targetBone {boneName=neck_01}}
    ]
    constraintWeights [1,1,1,1,1,1,1,1]
   }
   n37 = "Constraint - dyn_clip_shoulder_r_back_02_master (PPPPPPPP)"!(
    ComponentPose=@n36
    Alpha=1
    ConstraintWeights_0=0.008
    ConstraintWeights_1:float=0.0114
    ConstraintWeights_2:float=0.0114
    ConstraintWeights_3:float=0.0369
    ConstraintWeights_4:float=0.0594
    ConstraintWeights_5:float=0.1937
    ConstraintWeights_6:float=0.456
    ConstraintWeights_7:float=0.2232
   ) -> (Pose) {
    alphaBoolBlend {blendOption=Linear}
    boneToModify {boneName=dyn_clip_shoulder_r_back_02_master}
    constraintSetup [
     {targetBone {boneName=upperarm_r}}
     {targetBone {boneName=clavicle_l}}
     {targetBone {boneName=deform_deltoid_fr_r}}
     {targetBone {boneName=spine_02}}
     {targetBone {boneName=spine_03}}
     {targetBone {boneName=spine_04}}
     {targetBone {boneName=spine_05}}
     {targetBone {boneName=clavicle_r}}
    ]
    constraintWeights [1,1,1,1,1,1,1,1]
   }
   n38 = "Constraint - dyn_clip_back_l_01_master (PPPPPPPP)"!(
    ComponentPose=@n37
    Alpha=1
    ConstraintWeights_0=0.0009
    ConstraintWeights_1:float=0.141
    ConstraintWeights_2:float=0.2725
    ConstraintWeights_3:float=0.2724
    ConstraintWeights_4:float=0.1662
    ConstraintWeights_5:float=0.0854
    ConstraintWeights_6:float=0.0004
    ConstraintWeights_7:float=0.0611
   ) -> (Pose) {
    alphaBoolBlend {blendOption=Linear}
    boneToModify {boneName=dyn_clip_back_l_01_master}
    constraintSetup [
     {targetBone {boneName=clavicle_l}}
     {targetBone {boneName=spine_01}}
     {targetBone {boneName=spine_02}}
     {targetBone {boneName=spine_03}}
     {targetBone {boneName=spine_04}}
     {targetBone {boneName=spine_05}}
     {targetBone {boneName=upperarm_twist_01_l}}
     {targetBone {boneName=pelvis}}
    ]
    constraintWeights [1,1,1,1,1,1,1,1]
   }
   n39 = "Constraint - dyn_clip_back_l_02_master (PPPPPP)"!(
    ComponentPose=@n38
    Alpha=1
    ConstraintWeights_0=0.2485
    ConstraintWeights_1:float=0.2608
    ConstraintWeights_2:float=0.1865
    ConstraintWeights_3:float=0.0534
    ConstraintWeights_4:float=0.0204
    ConstraintWeights_5:float=0.2304
   ) -> (Pose) {
    alphaBoolBlend {blendOption=Linear}
    boneToModify {boneName=dyn_clip_back_l_02_master}
    constraintSetup [
     {targetBone {boneName=spine_01}}
     {targetBone {boneName=spine_02}}
     {targetBone {boneName=spine_03}}
     {targetBone {boneName=spine_04}}
     {targetBone {boneName=spine_05}}
     {targetBone {boneName=pelvis}}
    ]
    constraintWeights [1,1,1,1,1,1]
   }
   n40 = "Constraint - dyn_clip_back_r_01_master (PPPPPP)"!(
    ComponentPose=@n39
    Alpha=1
    ConstraintWeights_0=0.1473
    ConstraintWeights_1:float=0.2606
    ConstraintWeights_2:float=0.2517
    ConstraintWeights_3:float=0.1812
    ConstraintWeights_4:float=0.0793
    ConstraintWeights_5:float=0.0799
   ) -> (Pose) {
    alphaBoolBlend {blendOption=Linear}
    boneToModify {boneName=dyn_clip_back_r_01_master}
    constraintSetup [
     {targetBone {boneName=spine_01}}
     {targetBone {boneName=spine_02}}
     {targetBone {boneName=spine_03}}
     {targetBone {boneName=spine_04}}
     {targetBone {boneName=spine_05}}
     {targetBone {boneName=pelvis}}
    ]
    constraintWeights [1,1,1,1,1,1]
   }
   n41 = "Constraint - dyn_clip_back_r_02_master (PPPPPP)"!(
    ComponentPose=@n40
    Alpha=1
    ConstraintWeights_0=0.2321
    ConstraintWeights_1:float=0.2419
    ConstraintWeights_2:float=0.1795
    ConstraintWeights_3:float=0.0678
    ConstraintWeights_4:float=0.0234
    ConstraintWeights_5:float=0.2554
   ) -> (Pose) {
    alphaBoolBlend {blendOption=Linear}
    boneToModify {boneName=dyn_clip_back_r_02_master}
    constraintSetup [
     {targetBone {boneName=spine_01}}
     {targetBone {boneName=spine_02}}
     {targetBone {boneName=spine_03}}
     {targetBone {boneName=spine_04}}
     {targetBone {boneName=spine_05}}
     {targetBone {boneName=pelvis}}
    ]
    constraintWeights [1,1,1,1,1,1]
   }
   n42 = "Constraint - dyn_underarm_clip_l_master (PPPPPPPP)"!(
    ComponentPose=@n41
    Alpha=1
    ConstraintWeights_0=0.0094
    ConstraintWeights_1:float=0.1498
    ConstraintWeights_2:float=0.2512
    ConstraintWeights_3:float=0.2161
    ConstraintWeights_4:float=0.1781
    ConstraintWeights_5:float=0.0291
    ConstraintWeights_6:float=0.0113
    ConstraintWeights_7:float=0.155
   ) -> (Pose) {
    alphaBoolBlend {blendOption=Linear}
    boneToModify {boneName=dyn_underarm_clip_l_master}
    constraintSetup [
     {targetBone {boneName=clavicle_l}}
     {targetBone {boneName=spine_01}}
     {targetBone {boneName=spine_02}}
     {targetBone {boneName=spine_03}}
     {targetBone {boneName=spine_04}}
     {targetBone {boneName=spine_05}}
     {targetBone {boneName=upperarm_twist_01_l}}
     {targetBone {boneName=pelvis}}
    ]
    constraintWeights [1,1,1,1,1,1,1,1]
   }
   n43 = "Constraint - dyn_underarm_strap_l_master (PPPPPPPP)"!(
    ComponentPose=@n42
    Alpha=1
    ConstraintWeights_0=0.0183
    ConstraintWeights_1:float=0.1281
    ConstraintWeights_2:float=0.2567
    ConstraintWeights_3:float=0.2288
    ConstraintWeights_4:float=0.2245
    ConstraintWeights_5:float=0.0423
    ConstraintWeights_6:float=0.0192
    ConstraintWeights_7:float=0.0821
   ) -> (Pose) {
    alphaBoolBlend {blendOption=Linear}
    boneToModify {boneName=dyn_underarm_strap_l_master}
    constraintSetup [
     {targetBone {boneName=clavicle_l}}
     {targetBone {boneName=spine_01}}
     {targetBone {boneName=spine_02}}
     {targetBone {boneName=spine_03}}
     {targetBone {boneName=spine_04}}
     {targetBone {boneName=spine_05}}
     {targetBone {boneName=upperarm_twist_01_l}}
     {targetBone {boneName=pelvis}}
    ]
    constraintWeights [1,1,1,1,1,1,1,1]
   }
   n44 = "Constraint - dyn_underarm_strap_r_master (PPPPPPPP)"!(
    ComponentPose=@n43
    Alpha=1
    ConstraintWeights_0=0.0103
    ConstraintWeights_1:float=0.1315
    ConstraintWeights_2:float=0.2454
    ConstraintWeights_3:float=0.2087
    ConstraintWeights_4:float=0.2438
    ConstraintWeights_5:float=0.0477
    ConstraintWeights_6:float=0.0093
    ConstraintWeights_7:float=0.1032
   ) -> (Pose) {
    alphaBoolBlend {blendOption=Linear}
    boneToModify {boneName=dyn_underarm_strap_r_master}
    constraintSetup [
     {targetBone {boneName=upperarm_twist_01_r}}
     {targetBone {boneName=spine_01}}
     {targetBone {boneName=spine_02}}
     {targetBone {boneName=spine_03}}
     {targetBone {boneName=spine_04}}
     {targetBone {boneName=spine_05}}
     {targetBone {boneName=clavicle_r}}
     {targetBone {boneName=pelvis}}
    ]
    constraintWeights [1,1,1,1,1,1,1,1]
   }
   n45 = "Constraint - dyn_underarm_clip_r_master (PPPPPPPP)"!(
    ComponentPose=@n44
    Alpha=1
    ConstraintWeights_0=0.0072
    ConstraintWeights_1:float=0.143
    ConstraintWeights_2:float=0.2324
    ConstraintWeights_3:float=0.1949
    ConstraintWeights_4:float=0.1953
    ConstraintWeights_5:float=0.0343
    ConstraintWeights_6:float=0.0046
    ConstraintWeights_7:float=0.1882
   ) -> (Pose) {
    alphaBoolBlend {blendOption=Linear}
    boneToModify {boneName=dyn_underarm_clip_r_master}
    constraintSetup [
     {targetBone {boneName=upperarm_twist_01_r}}
     {targetBone {boneName=spine_01}}
     {targetBone {boneName=spine_02}}
     {targetBone {boneName=spine_03}}
     {targetBone {boneName=spine_04}}
     {targetBone {boneName=spine_05}}
     {targetBone {boneName=clavicle_r}}
     {targetBone {boneName=pelvis}}
    ]
    constraintWeights [1,1,1,1,1,1,1,1]
   }
   n46 = "Constraint - dyn_radio_base_master (PPPPPPPP)"!(
    ComponentPose=@n45
    Alpha=1
    ConstraintWeights_0=0.1735
    ConstraintWeights_1:float=0.0321
    ConstraintWeights_2:float=0.0231
    ConstraintWeights_3:float=0.2785
    ConstraintWeights_4:float=0.4067
    ConstraintWeights_5:float=0.01
    ConstraintWeights_6:float=0.0273
    ConstraintWeights_7:float=0.0487
   ) -> (Pose) {
    alphaBoolBlend {blendOption=Linear}
    boneToModify {boneName=dyn_radio_base_master}
    constraintSetup [
     {targetBone {boneName=clavicle_l}}
     {targetBone {boneName=deform_pectoral_fr_l}}
     {targetBone {boneName=spine_03}}
     {targetBone {boneName=spine_04}}
     {targetBone {boneName=spine_05}}
     {targetBone {boneName=upperarm_twist_01_l}}
     {targetBone {boneName=neck_02}}
     {targetBone {boneName=neck_01}}
    ]
    constraintWeights [1,1,1,1,1,1,1,1]
   }
   n47 = "Constraint - dyn_pouch_back_large_master (PPPPPPPP)"!(
    ComponentPose=@n46
    Alpha=1
    ConstraintWeights_0=0.0013
    ConstraintWeights_1:float=0.0116
    ConstraintWeights_2:float=0.0082
    ConstraintWeights_3:float=0.0049
    ConstraintWeights_4:float=0.0009
    ConstraintWeights_5:float=0.9699
    ConstraintWeights_6:float=0.0018
    ConstraintWeights_7:float=0.0014
   ) -> (Pose) {
    alphaBoolBlend {blendOption=Linear}
    boneToModify {boneName=dyn_pouch_back_large_master}
    constraintSetup [
     {targetBone {boneName=thigh_twist_01_r}}
     {targetBone {boneName=spine_01}}
     {targetBone {boneName=spine_02}}
     {targetBone {boneName=spine_03}}
     {targetBone {boneName=spine_04}}
     {targetBone {boneName=pelvis}}
     {targetBone {boneName=thigh_twist_01_l}}
     {targetBone {boneName=deform_thigh_bk_l}}
    ]
    constraintWeights [1,1,1,1,1,1,1,1]
   }
   n48 = "Constraint - dyn_pouch_04_master (PPPPPPPP)"!(
    ComponentPose=@n47
    Alpha=1
    ConstraintWeights_0=0.0021
    ConstraintWeights_1:float=0.0036
    ConstraintWeights_2:float=0.0292
    ConstraintWeights_3:float=0.0208
    ConstraintWeights_4:float=0.0126
    ConstraintWeights_5:float=0.0028
    ConstraintWeights_6:float=0.9281
    ConstraintWeights_7:float=0.0008
   ) -> (Pose) {
    alphaBoolBlend {blendOption=Linear}
    boneToModify {boneName=dyn_pouch_04_master}
    constraintSetup [
     {targetBone {boneName=deform_thigh_bk_r}}
     {targetBone {boneName=thigh_twist_01_r}}
     {targetBone {boneName=spine_01}}
     {targetBone {boneName=spine_02}}
     {targetBone {boneName=spine_03}}
     {targetBone {boneName=spine_04}}
     {targetBone {boneName=pelvis}}
     {targetBone {boneName=deform_thigh_fr_r}}
    ]
    constraintWeights [1,1,1,1,1,1,1,1]
   }
   n49 = "Constraint - dyn_pouch_05_master (PPPPPPPP)"!(
    ComponentPose=@n48
    Alpha=1
    ConstraintWeights_0=0.0039
    ConstraintWeights_1:float=0.0104
    ConstraintWeights_2:float=0.0359
    ConstraintWeights_3:float=0.0244
    ConstraintWeights_4:float=0.0153
    ConstraintWeights_5:float=0.0051
    ConstraintWeights_6:float=0.9
    ConstraintWeights_7:float=0.005
   ) -> (Pose) {
    alphaBoolBlend {blendOption=Linear}
    boneToModify {boneName=dyn_pouch_05_master}
    constraintSetup [
     {targetBone {boneName=deform_thigh_bk_r}}
     {targetBone {boneName=thigh_twist_01_r}}
     {targetBone {boneName=spine_01}}
     {targetBone {boneName=spine_02}}
     {targetBone {boneName=spine_03}}
     {targetBone {boneName=spine_04}}
     {targetBone {boneName=pelvis}}
     {targetBone {boneName=deform_thigh_fr_r}}
    ]
    constraintWeights [1,1,1,1,1,1,1,1]
   }
   n50 = "Constraint - dyn_holster_master (PPPPPPPPP)"!(
    ComponentPose=@n49
    Alpha=1
    ConstraintWeights_0=0.0001
    ConstraintWeights_1:float=0.1003
    ConstraintWeights_2:float=0.7802
    ConstraintWeights_3:float=0.0111
    ConstraintWeights_4:float=0.001
    ConstraintWeights_5:float=0.0001
    ConstraintWeights_6:float=0.0963
    ConstraintWeights_7:float=0.0097
    ConstraintWeights_8:float=0.0013
   ) -> (Pose) {
    alphaBoolBlend {blendOption=Linear}
    boneToModify {boneName=dyn_holster_master}
    constraintSetup [
     {targetBone {boneName=deform_knee_fr_01_r}}
     {targetBone {boneName=thigh_r}}
     {targetBone {boneName=thigh_twist_01_r}}
     {targetBone {boneName=deform_thigh_bk_r}}
     {targetBone {boneName=spine_01}}
     {targetBone {boneName=thigh_twist_01_l}}
     {targetBone {boneName=pelvis}}
     {targetBone {boneName=deform_thigh_fr_r}}
     {targetBone {boneName=deform_knee_bk_01_r}}
    ]
    constraintWeights [1,1,1,1,1,1,1,1,1]
   }
   n51 = "Constraint - dyn_pouch_radio_master (PPPPPPPP)"!(
    ComponentPose=@n50
    Alpha=1
    ConstraintWeights_0=0.0008
    ConstraintWeights_1:float=0.0313
    ConstraintWeights_2:float=0.0206
    ConstraintWeights_3:float=0.0113
    ConstraintWeights_4:float=0.0013
    ConstraintWeights_5:float=0.9239
    ConstraintWeights_6:float=0.0067
    ConstraintWeights_7:float=0.0042
   ) -> (Pose) {
    alphaBoolBlend {blendOption=Linear}
    boneToModify {boneName=dyn_pouch_radio_master}
    constraintSetup [
     {targetBone {boneName=deform_thigh_fr_l}}
     {targetBone {boneName=spine_01}}
     {targetBone {boneName=spine_02}}
     {targetBone {boneName=spine_03}}
     {targetBone {boneName=spine_04}}
     {targetBone {boneName=pelvis}}
     {targetBone {boneName=thigh_twist_01_l}}
     {targetBone {boneName=deform_thigh_bk_l}}
    ]
    constraintWeights [1,1,1,1,1,1,1,1]
   }
   n52 = "Constraint - dyn_pouch_03_master (PPPPPPPP)"!(
    ComponentPose=@n51
    Alpha=1
    ConstraintWeights_0=0.0136
    ConstraintWeights_1:float=0.0378
    ConstraintWeights_2:float=0.024
    ConstraintWeights_3:float=0.0159
    ConstraintWeights_4:float=0.0044
    ConstraintWeights_5:float=0.8663
    ConstraintWeights_6:float=0.0354
    ConstraintWeights_7:float=0.0025
   ) -> (Pose) {
    alphaBoolBlend {blendOption=Linear}
    boneToModify {boneName=dyn_pouch_03_master}
    constraintSetup [
     {targetBone {boneName=deform_thigh_fr_l}}
     {targetBone {boneName=spine_01}}
     {targetBone {boneName=spine_02}}
     {targetBone {boneName=spine_03}}
     {targetBone {boneName=spine_04}}
     {targetBone {boneName=pelvis}}
     {targetBone {boneName=thigh_twist_01_l}}
     {targetBone {boneName=deform_thigh_bk_l}}
    ]
    constraintWeights [1,1,1,1,1,1,1,1]
   }
   n53 = "Constraint - dyn_pouch_02_master (PPPPPPPP)"!(
    ComponentPose=@n52
    Alpha=1
    ConstraintWeights_0=0.0211
    ConstraintWeights_1:float=0.0307
    ConstraintWeights_2:float=0.0129
    ConstraintWeights_3:float=0.009
    ConstraintWeights_4:float=0.0029
    ConstraintWeights_5:float=0.872
    ConstraintWeights_6:float=0.0496
    ConstraintWeights_7:float=0.0018
   ) -> (Pose) {
    alphaBoolBlend {blendOption=Linear}
    boneToModify {boneName=dyn_pouch_02_master}
    constraintSetup [
     {targetBone {boneName=deform_thigh_fr_l}}
     {targetBone {boneName=spine_01}}
     {targetBone {boneName=spine_02}}
     {targetBone {boneName=spine_03}}
     {targetBone {boneName=spine_04}}
     {targetBone {boneName=pelvis}}
     {targetBone {boneName=thigh_twist_01_l}}
     {targetBone {boneName=deform_thigh_bk_l}}
    ]
    constraintWeights [1,1,1,1,1,1,1,1]
   }
   n54 = "Constraint - dyn_pouch_01_master (PPPPPPPP)"!(
    ComponentPose=@n53
    Alpha=1
    ConstraintWeights_0=0.024
    ConstraintWeights_1:float=0.0014
    ConstraintWeights_2:float=0.0302
    ConstraintWeights_3:float=0.0074
    ConstraintWeights_4:float=0.0055
    ConstraintWeights_5:float=0.002
    ConstraintWeights_6:float=0.8765
    ConstraintWeights_7:float=0.0529
   ) -> (Pose) {
    alphaBoolBlend {blendOption=Linear}
    boneToModify {boneName=dyn_pouch_01_master}
    constraintSetup [
     {targetBone {boneName=deform_thigh_fr_l}}
     {targetBone {boneName=thigh_twist_01_r}}
     {targetBone {boneName=spine_01}}
     {targetBone {boneName=spine_02}}
     {targetBone {boneName=spine_03}}
     {targetBone {boneName=spine_04}}
     {targetBone {boneName=pelvis}}
     {targetBone {boneName=thigh_twist_01_l}}
    ]
    constraintWeights [1,1,1,1,1,1,1,1]
   }
   n55 = "Constraint - dyn_zipper_master (PPPPPPP)"!(
    ComponentPose=@n54
    Alpha=1
    ConstraintWeights_0=0.0182
    ConstraintWeights_1:float=0.013
    ConstraintWeights_2:float=0.1331
    ConstraintWeights_3:float=0.7775
    ConstraintWeights_4:float=0.0166
    ConstraintWeights_5:float=0.0222
    ConstraintWeights_6:float=0.0195
   ) -> (Pose) {
    alphaBoolBlend {blendOption=Linear}
    boneToModify {boneName=dyn_zipper_master}
    constraintSetup [
     {targetBone {boneName=clavicle_l}}
     {targetBone {boneName=spine_03}}
     {targetBone {boneName=spine_04}}
     {targetBone {boneName=spine_05}}
     {targetBone {boneName=clavicle_r}}
     {targetBone {boneName=neck_02}}
     {targetBone {boneName=neck_01}}
    ]
    constraintWeights [1,1,1,1,1,1,1]
   }
   n56 = "Constraint - dyn_radio_top_cable_05 (PPPPPPPP)"!(
    ComponentPose=@n55
    Alpha=1
    ConstraintWeights_0=0.2544
    ConstraintWeights_1:float=0.0116
    ConstraintWeights_2:float=0.0439
    ConstraintWeights_3:float=0.4923
    ConstraintWeights_4:float=0.0187
    ConstraintWeights_5:float=0.0704
    ConstraintWeights_6:float=0.0996
    ConstraintWeights_7:float=0.009
   ) -> (Pose) {
    alphaBoolBlend {blendOption=Linear}
    boneToModify {boneName=dyn_radio_top_cable_05}
    constraintSetup [
     {targetBone {boneName=clavicle_l}}
     {targetBone {boneName=dyn_l_shoulder_strap}}
     {targetBone {boneName=spine_04}}
     {targetBone {boneName=spine_05}}
     {targetBone {boneName=upperarm_twist_01_l}}
     {targetBone {boneName=neck_02}}
     {targetBone {boneName=neck_01}}
     {targetBone {boneName=dyn_neck}}
    ]
    constraintWeights [1,1,1,1,1,1,1,1]
   }
   n57 = "Constraint - dyn_radio_cable_10 (P)"(ComponentPose=@n56, Alpha=1, ConstraintWeights_0=1) {
    alphaBoolBlend {blendOption=Linear}
    boneToModify {boneName=dyn_radio_cable_10}
    constraintSetup [{targetBone {boneName=dyn_pouch_radio_wire_base}}]
    constraintWeights [1]
   }
   n58 = Look At - Bone: dyn_radio_top_cable_lookat(ComponentPose=@n57, Alpha=1) {
    alphaBoolBlend {blendOption=Linear}
    boneToModify {boneName=dyn_radio_top_cable_lookat}
    lookAtLocation {x=0}
    lookAtTarget {boneReference {boneName=dyn_radio_base}}
    lookAt_Axis {axis=<1 0>}
   }
   n59 = "Transform (Modify) Bone - Bone: dyn_pouch_01_master"!(ComponentPose=@n58, bAlphaBoolEnabled:bool=@n30, Translation="4.747,-2.972,3.921", Rotation="-11.419,-78.874,-53.478", Scale="1,1,1") -> (Pose) {
    alphaInputType=Bool
    rotationMode=BMM_Additive
    translationMode=BMM_Additive
    alphaBoolBlend {blendInTime=0.1, blendOption=Linear, blendOutTime=0.1}
    boneToModify {boneName=dyn_pouch_01_master}
    rotation=<rot -11.419 -78.874 -53.478>
    translation=<4.747 -2.972 3.921>
   }
   n60 = "Transform (Modify) Bone - Bone: dyn_pouch_02_master"!(ComponentPose=@n59, bAlphaBoolEnabled:bool=@n30, Translation="2.547,-6.482,0.000", Rotation="12.776,-29.313,16.437", Scale="1,1,1") -> (Pose) {
    alphaInputType=Bool
    rotationMode=BMM_Additive
    translationMode=BMM_Additive
    alphaBoolBlend {blendInTime=0.1, blendOption=Linear, blendOutTime=0.1}
    boneToModify {boneName=dyn_pouch_02_master}
    rotation=<rot 12.776 -29.313 16.4372>
    translation=<2.547 -6.482>
   }
   n61 = "Transform (Modify) Bone - Bone: dyn_pouch_03_master"!(ComponentPose=@n60, bAlphaBoolEnabled:bool=@n30, Translation="0.144,-6.895,0.000", Rotation="3.002,-36.954,9.743", Scale="1,1,1") -> (Pose) {
    alphaInputType=Bool
    rotationMode=BMM_Additive
    translationMode=BMM_Additive
    alphaBoolBlend {blendInTime=0.1, blendOption=Linear, blendOutTime=0.1}
    boneToModify {boneName=dyn_pouch_03_master}
    rotation=<rot 3.002 -36.9539 9.7431>
    translation=<0.144 -6.895>
   }
   n62 = Get crouchmove() -> (crouchmove:bool)
   n63 = "Transform (Modify) Bone - Bone: dyn_pouch_01_master"(ComponentPose=@n61, bAlphaBoolEnabled=@n62, Translation="0,0,0", Rotation="7.182,41.065,-8.164", Scale="1,1,1") {
    alphaInputType=Bool
    rotationMode=BMM_Additive
    translationMode=BMM_Additive
    alphaBoolBlend {blendInTime=0.5, blendOption=Linear, blendOutTime=0.1}
    boneToModify {boneName=dyn_pouch_01_master}
    rotation=<rot 7.182 41.065 -8.164>
   }
   n64 = "Transform (Modify) Bone - Bone: dyn_pouch_02_master"(ComponentPose=@n63, bAlphaBoolEnabled=@n62, Translation="0,0,3", Rotation="35,0,0", Scale="1,1,1") {
    alphaInputType=Bool
    rotationMode=BMM_Additive
    translationMode=BMM_Additive
    alphaBoolBlend {blendInTime=0.5, blendOption=Linear, blendOutTime=0.1}
    boneToModify {boneName=dyn_pouch_02_master}
    rotation {pitch=35}
    translation {z=3}
   }
   n65 = "Transform (Modify) Bone - Bone: dyn_pouch_03_master"(ComponentPose=@n64, bAlphaBoolEnabled=@n62, Translation="0,0,3", Rotation="35,0.000,0.000", Scale="1,1,1") {
    alphaInputType=Bool
    rotationMode=BMM_Additive
    translationMode=BMM_Additive
    alphaBoolBlend {blendInTime=0.5, blendOption=Linear, blendOutTime=0.1}
    boneToModify {boneName=dyn_pouch_03_master}
    rotation {pitch=35}
    translation {z=3}
   }
   n66 = Anim Dynamics - Bone: dyn_pouch_back_large(ComponentPose=@n65, Alpha=0.45, ExternalForce="0,0,0") {
    angularDampingOverride=1
    angularSpringConstant=100
    bAngularSpring=true
    bGravityOverrideInSimSpace=true
    bOverrideAngularDamping=true
    bUseGravityOverride=true
    gravityScale=0.35
    simulationSpace=RootRelative
    alphaBoolBlend {blendOption=Linear}
    boundBone {boneName=dyn_pouch_back_large}
    componentAppliedLinearAccClamp=<3500 3500 3500>
    componentLinearAccScale=<0.3 0.3 0.3>
    componentLinearVelScale=<0.3 0.3 0.3>
    gravityOverride {z=-980}
    physicsBodyDefinitions [{
      sphereCollisionRadius=0
      boundBone {boneName=dyn_pouch_back_large}
      boxExtents=<12 12 6>
      constraintSetup {angularLimitsMax=<5 2 1>, angularLimitsMin=<-5 -35 -1>, angularTarget=<1 0 0>}
      localJointOffset=<3 0 0>
     }]
   }
   n67 = Anim Dynamics - Bone: dyn_pouch_04(ComponentPose=@n66, Alpha=0.5, ExternalForce="0,0,0") {
    angularDampingOverride=0.98
    angularSpringConstant=75
    bAngularSpring=true
    bGravityOverrideInSimSpace=true
    bOverrideAngularDamping=true
    bUseGravityOverride=true
    gravityScale=1.2
    simulationSpace=RootRelative
    alphaBoolBlend {blendOption=Linear}
    boundBone {boneName=dyn_pouch_04}
    componentAppliedLinearAccClamp=<3500 3500 3500>
    componentLinearAccScale=<0.3 0.3 0.3>
    componentLinearVelScale=<0.3 0.3 0.3>
    gravityOverride {z=-980}
    physicsBodyDefinitions [{
      sphereCollisionRadius=0
      boundBone {boneName=dyn_pouch_04}
      boxExtents=<12 13 7>
      constraintSetup {angularLimitsMax=<1 2 -5>, angularLimitsMin=<-1 -35 -45>, angularTarget=<0 0 0>}
      localJointOffset=<1 0 0>
     }]
   }
   n68 = Anim Dynamics - Bone: dyn_pouch_05(ComponentPose=@n67, Alpha=0.2, ExternalForce="0,0,0") {
    angularDampingOverride=0.98
    angularSpringConstant=50
    bAngularSpring=true
    bGravityOverrideInSimSpace=true
    bOverrideAngularDamping=true
    bUseGravityOverride=true
    gravityScale=1.2
    simulationSpace=RootRelative
    alphaBoolBlend {blendOption=Linear}
    boundBone {boneName=dyn_pouch_05}
    componentAppliedLinearAccClamp=<3500 3500 3500>
    componentLinearAccScale=<0.3 0.3 0.3>
    componentLinearVelScale=<0.3 0.3 0.3>
    gravityOverride {z=-980}
    physicsBodyDefinitions [{
      sphereCollisionRadius=0
      boundBone {boneName=dyn_pouch_05}
      boxExtents=<12 12 8>
      constraintSetup {angularLimitsMax=<1 2 -5>, angularLimitsMin=<-1 -35 -45>, angularTarget=<0 0 0>}
      localJointOffset=<1 0.1 0.1>
     }]
   }
   n69 = Anim Dynamics - Bone: dyn_pouch_01(ComponentPose=@n68, Alpha=0.15, ExternalForce="0,0,0") {
    angularDampingOverride=1
    angularSpringConstant=75
    bAngularSpring=true
    bGravityOverrideInSimSpace=true
    bOverrideAngularDamping=true
    bUseGravityOverride=true
    gravityScale=0.35
    simulationSpace=RootRelative
    alphaBoolBlend {blendOption=Linear}
    boundBone {boneName=dyn_pouch_01}
    componentAppliedLinearAccClamp=<5000 5000 5000>
    componentLinearAccScale=<0.5 0.5 0.5>
    componentLinearVelScale=<0.5 0.5 0.5>
    gravityOverride {z=-980}
    physicsBodyDefinitions [{
      sphereCollisionRadius=0
      boundBone {boneName=dyn_pouch_01}
      boxExtents=<22 12 12>
      constraintSetup {angularLimitsMax=<40 2 40>, angularLimitsMin=<-40 -40 -40>, angularTarget=<1 0 0>}
      localJointOffset=<4 0 0.1>
     }]
   }
   n70 = Anim Dynamics - Bone: dyn_pouch_02(ComponentPose=@n69, Alpha=0.2, ExternalForce="0,0,0") {
    angularDampingOverride=0.999
    angularSpringConstant=150
    bAngularSpring=true
    bGravityOverrideInSimSpace=true
    bLinearSpring=true
    bOverrideAngularDamping=true
    bUseGravityOverride=true
    lODThreshold=1
    simulationSpace=BoneRelative
    alphaBoolBlend {blendOption=Linear}
    boundBone {boneName=dyn_pouch_02}
    componentLinearAccScale=<0.25 0.258 0.258>
    componentLinearVelScale=<0.25 0.25 0.25>
    gravityOverride {z=-1500}
    physicsBodyDefinitions [{
      sphereCollisionRadius=0
      boundBone {boneName=dyn_pouch_02}
      boxExtents=<25 15 15>
      constraintSetup {angularTargetAxis=AxisZ, angularLimitsMax=<15 2 40>, angularLimitsMin=<-15 -30 -40>, angularTarget=<0 0 1>}
      localJointOffset=<2 0 0>
     }]
    relativeSpaceBone {boneName=root}
   }
   n71 = Anim Dynamics - Bone: dyn_pouch_03(ComponentPose=@n70, Alpha=0.25, ExternalForce="0,0,0") {
    angularDampingOverride=1
    angularSpringConstant=75
    bAngularSpring=true
    bGravityOverrideInSimSpace=true
    bOverrideAngularDamping=true
    bUseGravityOverride=true
    gravityScale=0.35
    simulationSpace=RootRelative
    alphaBoolBlend {blendOption=Linear}
    boundBone {boneName=dyn_pouch_03}
    componentAppliedLinearAccClamp=<5000 5000 5000>
    componentLinearAccScale=<0.5 0.5 0.5>
    componentLinearVelScale=<0.5 0.5 0.5>
    gravityOverride {z=-980}
    physicsBodyDefinitions [{
      sphereCollisionRadius=0
      boundBone {boneName=dyn_pouch_03}
      boxExtents=<20 10 10>
      constraintSetup {angularLimitsMax=<40 2 40>, angularLimitsMin=<-40 -40 -40>, angularTarget=<1 0 0>}
      localJointOffset=<3 0 0>
     }]
   }
   n72 = Anim Dynamics - Bone: dyn_pouch_radio(ComponentPose=@n71, Alpha=0.15, ExternalForce="0,0,0") {
    angularDampingOverride=1
    angularSpringConstant=75
    bAngularSpring=true
    bGravityOverrideInSimSpace=true
    bOverrideAngularDamping=true
    bUseGravityOverride=true
    gravityScale=0.35
    simulationSpace=RootRelative
    alphaBoolBlend {blendOption=Linear}
    boundBone {boneName=dyn_pouch_radio}
    componentAppliedLinearAccClamp=<5000 5000 5000>
    componentLinearAccScale=<0.5 0.5 0.5>
    componentLinearVelScale=<0.5 0.5 0.5>
    gravityOverride {z=-980}
    physicsBodyDefinitions [{
      sphereCollisionRadius=0
      boundBone {boneName=dyn_pouch_radio}
      boxExtents=<22 12 12>
      constraintSetup {angularLimitsMax=<40 2 40>, angularLimitsMin=<-40 -40 -40>, angularTarget=<1 0 0>}
      localJointOffset=<4 0 0.1>
     }]
   }
   n73 = Anim Dynamics - Bone: dyn_clip_shoulder_r_back_02(ComponentPose=@n72, Alpha=0.5, ExternalForce="0,0,0") {
    angularDampingOverride=1
    angularSpringConstant=75
    bAngularSpring=true
    bGravityOverrideInSimSpace=true
    bOverrideAngularDamping=true
    bUseGravityOverride=true
    gravityScale=0.35
    simulationSpace=RootRelative
    alphaBoolBlend {blendOption=Linear}
    boundBone {boneName=dyn_clip_shoulder_r_back_02}
    componentAppliedLinearAccClamp=<5000 5000 5000>
    componentLinearAccScale=<0.5 0.5 0.5>
    componentLinearVelScale=<0.5 0.5 0.5>
    gravityOverride {z=-980}
    physicsBodyDefinitions [{
      sphereCollisionRadius=0
      boundBone {boneName=dyn_clip_shoulder_r_back_02}
      boxExtents=<20 10 10>
      constraintSetup {angularLimitsMax=<15 -2 5>, angularLimitsMin=<-15 -40 -5>, angularTarget=<1 0 0>}
      localJointOffset=<2 0 0>
     }]
   }
   n74 = Anim Dynamics - Bone: dyn_clip_shoulder_l_back_02(ComponentPose=@n73, Alpha=0.5, ExternalForce="0,0,0") {
    angularDampingOverride=1
    angularSpringConstant=75
    bAngularSpring=true
    bGravityOverrideInSimSpace=true
    bOverrideAngularDamping=true
    bUseGravityOverride=true
    gravityScale=0.35
    simulationSpace=RootRelative
    alphaBoolBlend {blendOption=Linear}
    boundBone {boneName=dyn_clip_shoulder_l_back_02}
    componentAppliedLinearAccClamp=<5000 5000 5000>
    componentLinearAccScale=<0.5 0.5 0.5>
    componentLinearVelScale=<0.5 0.5 0.5>
    gravityOverride {z=-980}
    physicsBodyDefinitions [{
      sphereCollisionRadius=0
      boundBone {boneName=dyn_clip_shoulder_l_back_02}
      boxExtents=<20 10 10>
      constraintSetup {angularLimitsMax=<15 -2 5>, angularLimitsMin=<-15 -40 -5>, angularTarget=<1 0 0>}
      localJointOffset=<2 0 0>
     }]
   }
   n75 = Anim Dynamics - Bone: dyn_clip_shoulder_r_back_01(ComponentPose=@n74, Alpha=0.5, ExternalForce="0,0,0") {
    angularDampingOverride=1
    angularSpringConstant=75
    bAngularSpring=true
    bGravityOverrideInSimSpace=true
    bOverrideAngularDamping=true
    bUseGravityOverride=true
    gravityScale=0.35
    simulationSpace=RootRelative
    alphaBoolBlend {blendOption=Linear}
    boundBone {boneName=dyn_clip_shoulder_r_back_01}
    componentAppliedLinearAccClamp=<5000 5000 5000>
    componentLinearAccScale=<0.5 0.5 0.5>
    componentLinearVelScale=<0.5 0.5 0.5>
    gravityOverride {z=-980}
    physicsBodyDefinitions [{
      sphereCollisionRadius=0
      boundBone {boneName=dyn_clip_shoulder_r_back_01}
      boxExtents=<20 10 10>
      constraintSetup {angularLimitsMax=<40 -5 5>, angularLimitsMin=<-40 -40 -5>, angularTarget=<1 0 0>}
      localJointOffset=<2 0 0>
     }]
   }
   n76 = Anim Dynamics - Bone: dyn_clip_shoulder_l_back_01(ComponentPose=@n75, Alpha=0.5, ExternalForce="0,0,0") {
    angularDampingOverride=1
    angularSpringConstant=75
    bAngularSpring=true
    bGravityOverrideInSimSpace=true
    bOverrideAngularDamping=true
    bUseGravityOverride=true
    gravityScale=0.35
    simulationSpace=RootRelative
    alphaBoolBlend {blendOption=Linear}
    boundBone {boneName=dyn_clip_shoulder_l_back_01}
    componentAppliedLinearAccClamp=<5000 5000 5000>
    componentLinearAccScale=<0.5 0.5 0.5>
    componentLinearVelScale=<0.5 0.5 0.5>
    gravityOverride {z=-980}
    physicsBodyDefinitions [{
      sphereCollisionRadius=0
      boundBone {boneName=dyn_clip_shoulder_l_back_01}
      boxExtents=<20 10 10>
      constraintSetup {angularLimitsMax=<40 2 5>, angularLimitsMin=<-40 -40 -5>, angularTarget=<1 0 0>}
      localJointOffset=<2 0 0>
     }]
   }
   n77 = Anim Dynamics - Bone: dyn_clip_back_r_01(ComponentPose=@n76, Alpha=0.75, ExternalForce="0,0,0") {
    angularDampingOverride=1
    angularSpringConstant=75
    bAngularSpring=true
    bGravityOverrideInSimSpace=true
    bOverrideAngularDamping=true
    bUseGravityOverride=true
    gravityScale=0.35
    simulationSpace=RootRelative
    alphaBoolBlend {blendOption=Linear}
    boundBone {boneName=dyn_clip_back_r_01}
    componentAppliedLinearAccClamp=<5000 5000 5000>
    componentLinearAccScale=<0.5 0.5 0.5>
    componentLinearVelScale=<0.5 0.5 0.5>
    gravityOverride {z=-980}
    physicsBodyDefinitions [{
      sphereCollisionRadius=0
      boundBone {boneName=dyn_clip_back_r_01}
      boxExtents=<20 10 10>
      constraintSetup {angularLimitsMax=<5 2 1>, angularLimitsMin=<-5 -40 -1>, angularTarget=<1 0 0>}
      localJointOffset=<2 0 0>
     }]
   }
   n78 = Anim Dynamics - Bone: dyn_clip_back_l_01(ComponentPose=@n77, Alpha=0.75, ExternalForce="0,0,0") {
    angularDampingOverride=1
    angularSpringConstant=75
    bAngularSpring=true
    bGravityOverrideInSimSpace=true
    bOverrideAngularDamping=true
    bUseGravityOverride=true
    gravityScale=0.35
    simulationSpace=RootRelative
    alphaBoolBlend {blendOption=Linear}
    boundBone {boneName=dyn_clip_back_l_01}
    componentAppliedLinearAccClamp=<5000 5000 5000>
    componentLinearAccScale=<0.5 0.5 0.5>
    componentLinearVelScale=<0.5 0.5 0.5>
    gravityOverride {z=-980}
    physicsBodyDefinitions [{
      sphereCollisionRadius=0
      boundBone {boneName=dyn_clip_back_l_01}
      boxExtents=<20 10 10>
      constraintSetup {angularLimitsMax=<5 2 1>, angularLimitsMin=<-5 -40 -1>, angularTarget=<1 0 0>}
      localJointOffset=<2 0 0>
     }]
   }
   n79 = Anim Dynamics - Bone: dyn_clip_back_l_02(ComponentPose=@n78, Alpha=0.75, ExternalForce="0,0,0") {
    angularDampingOverride=1
    angularSpringConstant=75
    bAngularSpring=true
    bGravityOverrideInSimSpace=true
    bOverrideAngularDamping=true
    bUseGravityOverride=true
    gravityScale=0.35
    simulationSpace=RootRelative
    alphaBoolBlend {blendOption=Linear}
    boundBone {boneName=dyn_clip_back_l_02}
    componentAppliedLinearAccClamp=<5000 5000 5000>
    componentLinearAccScale=<0.5 0.5 0.5>
    componentLinearVelScale=<0.5 0.5 0.5>
    gravityOverride {z=-980}
    physicsBodyDefinitions [{
      sphereCollisionRadius=0
      boundBone {boneName=dyn_clip_back_l_02}
      boxExtents=<20 10 10>
      constraintSetup {angularLimitsMax=<5 2 1>, angularLimitsMin=<-5 -40 -1>, angularTarget=<1 0 0>}
      localJointOffset=<2 0 0>
     }]
   }
   n80 = Anim Dynamics - Bone: dyn_clip_back_r_02(ComponentPose=@n79, Alpha=0.75, ExternalForce="0,0,0") {
    angularDampingOverride=1
    angularSpringConstant=75
    bAngularSpring=true
    bGravityOverrideInSimSpace=true
    bOverrideAngularDamping=true
    bUseGravityOverride=true
    gravityScale=0.35
    simulationSpace=RootRelative
    alphaBoolBlend {blendOption=Linear}
    boundBone {boneName=dyn_clip_back_r_02}
    componentAppliedLinearAccClamp=<5000 5000 5000>
    componentLinearAccScale=<0.5 0.5 0.5>
    componentLinearVelScale=<0.5 0.5 0.5>
    gravityOverride {z=-980}
    physicsBodyDefinitions [{
      sphereCollisionRadius=0
      boundBone {boneName=dyn_clip_back_r_02}
      boxExtents=<20 10 10>
      constraintSetup {angularLimitsMax=<5 2 1>, angularLimitsMin=<-5 -40 -1>, angularTarget=<1 0 0>}
      localJointOffset=<2 0 0>
     }]
   }
   n81 = Anim Dynamics - Bone: dyn_underarm_clip_r(ComponentPose=@n80, Alpha=0.75, ExternalForce="0,0,0") {
    angularDampingOverride=1
    angularSpringConstant=75
    bAngularSpring=true
    bGravityOverrideInSimSpace=true
    bOverrideAngularDamping=true
    bUseGravityOverride=true
    gravityScale=0.35
    simulationSpace=RootRelative
    alphaBoolBlend {blendOption=Linear}
    boundBone {boneName=dyn_underarm_clip_r}
    componentAppliedLinearAccClamp=<5000 5000 5000>
    componentLinearAccScale=<0.5 0.5 0.5>
    componentLinearVelScale=<0.5 0.5 0.5>
    gravityOverride {z=-980}
    physicsBodyDefinitions [{
      sphereCollisionRadius=0
      boundBone {boneName=dyn_underarm_clip_r}
      boxExtents=<15 5 5>
      constraintSetup {angularLimitsMax=<5 -6 1>, angularLimitsMin=<-5 -40 -1>, angularTarget=<1 0 0>}
      localJointOffset=<3 0 0>
     }]
   }
   n82 = Anim Dynamics - Bone: dyn_underarm_clip_l(ComponentPose=@n81, Alpha=0.75, ExternalForce="0,0,0") {
    angularDampingOverride=1
    angularSpringConstant=75
    bAngularSpring=true
    bGravityOverrideInSimSpace=true
    bOverrideAngularDamping=true
    bUseGravityOverride=true
    gravityScale=0.35
    simulationSpace=RootRelative
    alphaBoolBlend {blendOption=Linear}
    boundBone {boneName=dyn_underarm_clip_l}
    componentAppliedLinearAccClamp=<5000 5000 5000>
    componentLinearAccScale=<0.5 0.5 0.5>
    componentLinearVelScale=<0.5 0.5 0.5>
    gravityOverride {z=-980}
    physicsBodyDefinitions [{
      sphereCollisionRadius=0
      boundBone {boneName=dyn_underarm_clip_l}
      boxExtents=<15 5 5>
      constraintSetup {angularLimitsMax=<5 2 1>, angularLimitsMin=<-5 -40 -1>, angularTarget=<1 0 0>}
      localJointOffset=<3 0 0>
     }]
   }
   n83 = Anim Dynamics - Bone: dyn_underarm_strap_l(ComponentPose=@n82, Alpha=0.75, ExternalForce="0,0,0") {
    angularDampingOverride=1
    angularSpringConstant=75
    bAngularSpring=true
    bGravityOverrideInSimSpace=true
    bOverrideAngularDamping=true
    bUseGravityOverride=true
    gravityScale=0.35
    simulationSpace=RootRelative
    alphaBoolBlend {blendOption=Linear}
    boundBone {boneName=dyn_underarm_strap_l}
    componentAppliedLinearAccClamp=<5000 5000 5000>
    componentLinearAccScale=<0.5 0.5 0.5>
    componentLinearVelScale=<0.5 0.5 0.5>
    gravityOverride {z=-980}
    physicsBodyDefinitions [{
      sphereCollisionRadius=0
      boundBone {boneName=dyn_underarm_strap_l}
      boxExtents=<15 5 5>
      constraintSetup {angularLimitsMax=<5 2 1>, angularLimitsMin=<-5 -40 -1>, angularTarget=<1 0 0>}
      localJointOffset=<3 0 0>
     }]
   }
   n84 = Anim Dynamics - Bone: dyn_underarm_strap_r(ComponentPose=@n83, Alpha=0.75, ExternalForce="0,0,0") {
    angularDampingOverride=1
    angularSpringConstant=75
    bAngularSpring=true
    bGravityOverrideInSimSpace=true
    bOverrideAngularDamping=true
    bUseGravityOverride=true
    gravityScale=0.35
    simulationSpace=RootRelative
    alphaBoolBlend {blendOption=Linear}
    boundBone {boneName=dyn_underarm_strap_r}
    componentAppliedLinearAccClamp=<5000 5000 5000>
    componentLinearAccScale=<0.5 0.5 0.5>
    componentLinearVelScale=<0.5 0.5 0.5>
    gravityOverride {z=-980}
    physicsBodyDefinitions [{
      sphereCollisionRadius=0
      boundBone {boneName=dyn_underarm_strap_r}
      boxExtents=<15 5 5>
      constraintSetup {angularLimitsMax=<5 2 1>, angularLimitsMin=<-5 -40 -1>, angularTarget=<1 0 0>}
      localJointOffset=<3 0 0>
     }]
   }
   n85 = Anim Dynamics - Bone: dyn_zipper!(ComponentPose=@n84, Alpha=0.25, GravityOverride:Vector="0,0,0", ExternalForce) -> (Pose) {
    angularDampingOverride=1
    angularSpringConstant=1500
    bAngularSpring=true
    bGravityOverrideInSimSpace=true
    bOverrideAngularDamping=true
    gravityScale=0.45
    simulationSpace=RootRelative
    alphaBoolBlend {blendOption=Linear}
    alphaScaleBias {scale=2}
    boundBone {boneName=dyn_zipper}
    componentAppliedLinearAccClamp=<2500 2500 2500>
    componentLinearAccScale=<1 1 1>
    componentLinearVelScale=<1 1 1>
    physicsBodyDefinitions [{
      sphereCollisionRadius=0
      boundBone {boneName=dyn_zipper}
      boxExtents=<10 5 5>
      constraintSetup {angularLimitsMax=<15 5 15>, angularLimitsMin=<-15 -35 -15>, angularTarget=<0 0 0>}
      localJointOffset=<2 0 0>
     }]
    relativeSpaceBone {boneName=root}
   }
   n86 = RigidBody!(ComponentPose=@n85, Alpha, ExternalForce) -> (Pose) {
    bClampLinearTranslationLimitToRefPose=true
    bOverrideWorldGravity=true
    overridePhysicsAsset=&/M_MED_Relish_Physics
    simulationSpace=BaseBoneSpace
    alphaBoolBlend {blendOption=Linear}
    baseBoneRef {boneName=spine_04}
    componentAppliedLinearAccClamp=<300 300 300>
    componentLinearAccScale=<0.25 0.25 0.25>
    componentLinearVelScale=<0.25 0.25 0.25>
    overrideWorldGravity {z=-980}
   }
   n87 = RigidBody(ComponentPose=@n86) {
    bClampLinearTranslationLimitToRefPose=true
    bOverrideWorldGravity=true
    overridePhysicsAsset=&/M_MED_Relish_top_Physics
    simulationSpace=BaseBoneSpace
    alphaBoolBlend {blendOption=Linear}
    baseBoneRef {boneName=spine_04}
    componentAppliedLinearAccClamp=<300 300 300>
    componentLinearAccScale=<0.25 0.25 0.25>
    componentLinearVelScale=<0.25 0.25 0.25>
    overrideWorldGravity {z=-980}
   }
   Output Pose(Result=Component To Local(ComponentPose=@n87)) {layerGroup=None}
  ]}
 EventGraph {nodes [
   Event BlueprintUpdateAnimation()
   > n1 = Sequence()
   n2 = Get Skel_Mesh()
   > Set thigh_rot_l(thigh_rot_l=GetAlphaFromReferenceTransformDelta(SkeletalMeshComponent=@n2, BoneName=thigh_l, BaseName=pelvis, InputRange=(X=-20,Y=-115), TwistAxis="1,0,0"))
   > Set thigh_rot_r(thigh_rot_r=GetAlphaFromReferenceTransformDelta(SkeletalMeshComponent=@n2, BoneName=thigh_r, BaseName=pelvis, InputRange=(X=-20,Y=-100), TwistAxis="1,0,0"))
   > Set thigh_rot_l_2(thigh_rot_l_2:double=GetAlphaFromReferenceTransformDelta(SkeletalMeshComponent=@n2, BoneName=thigh_l, BaseName=pelvis, DeltaComponentType=EulerX, InputRange=(X=-20,Y=-115), TwistAxis="1,0,0")) -> (Output_Get:double)
   > Set thigh_rot_r_2(thigh_rot_r_2:double=GetAlphaFromReferenceTransformDelta(SkeletalMeshComponent=@n2, BoneName=thigh_r, BaseName=pelvis, DeltaComponentType=EulerX, InputRange=(X=-20,Y=-100), TwistAxis="1,0,0")) -> (Output_Get:double)
   n3 = Branch(execute=@n1.then_1, Condition=AND Boolean(A=Get bIsCrouching(), B=Get bIsMoving2D()))
   > Set crouchmove(crouchmove:bool=true) -> (Output_Get:bool)
   Set crouchmove(execute=@n3.else, crouchmove=false)
   Event BlueprintInitializeAnimation()
   > Set Skel_Mesh(Skel_Mesh=GetOwnerSkeletalMeshComponent())
  ]}
]
variables [
 crouchmove {type=boolean}
 Skel_Mesh {type=SkeletalMeshComponent}
 thigh_rot_l {type=number}
 thigh_rot_l_2 {type=number}
 thigh_rot_r {type=number}
 thigh_rot_r_2 {type=number}
]
