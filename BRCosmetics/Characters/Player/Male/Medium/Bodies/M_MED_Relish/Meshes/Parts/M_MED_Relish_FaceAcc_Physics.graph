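# Physics asset for the face accessory mesh of the M_MED_Relish male player cosmetic body, covering a headset that hangs off the head.
# Capsule bodies sit on the neck, head, headset master bone, three headset segments and one radio top cable bone.
# Constraints chain the headset segments through the master to the head with very tight swing and twist limits, while head-to-neck rotation is free.
# The radio cable body has no constraint.
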PhysicsAsset /BRCosmetics/Characters/Player/Male/Medium/Bodies/M_MED_Relish/Meshes/Parts/M_MED_Relish_FaceAcc_Physics
bodies {
 dyn_headset_01 [{shapeType=Capsule, length=1.987, radius=0.7968, shapeName=dyn_headset_01_capsule, center=<1.5 -0.5 0>, rotation=<rot 90 0 17.794>}]
 dyn_headset_02 [{shapeType=Capsule, length=2.2486, radius=0.9867, shapeName=dyn_headset_02_capsule, center=<2 0 0>, rotation=<rot 90 0 0>}]
 dyn_headset_03 [{shapeType=Capsule, length=0.7172, radius=0.505, shapeName=dyn_headset_03_capsule, center=<1 0 0>, rotation=<rot 90 0 0>}]
 dyn_headset_MSTR [{shapeType=Capsule, length=0.505, radius=0.505, shapeName=dyn_headset_MSTR_capsule, rotation=<rot 0 -90 0>}]
 dyn_radio_top_cable_05 [{
   shapeType=Capsule
   length=0.505
   radius=0.505
   shapeName=dyn_radio_top_cable_05_capsule
   center=<-0.4373 -3.0043 0.5788>
   rotation=<rot 0 -57.7672 -72.462>
  }]
 head [{shapeType=Capsule, length=4.4115, radius=5.8829, shapeName=head_capsule, center=<0 2.6226 0>, rotation=<rot 90 0 0>}]
 neck_02 [{shapeType=Capsule, length=3.0286, radius=6.1724, shapeName=neck_02_capsule, rotation=<rot 90 0 0>}]
}
constraints [
 {
  bone1Name=head
  bone2Name=neck_02
  swing1LimitDegrees=45
  swing1Motion=Free
  swing2LimitDegrees=45
  swing2Motion=Free
  twistLimitDegrees=45
  twistMotion=Free
 }
 {bone1Name=dyn_headset_MSTR, bone2Name=head, swing1LimitDegrees=3, swing2LimitDegrees=3, twistLimitDegrees=1}
 {bone1Name=dyn_headset_01, bone2Name=dyn_headset_MSTR, swing1LimitDegrees=3, swing2LimitDegrees=3, twistLimitDegrees=1}
 {bone1Name=dyn_headset_02, bone2Name=dyn_headset_01, swing1LimitDegrees=3, swing2LimitDegrees=3, twistLimitDegrees=1}
 {bone1Name=dyn_headset_03, bone2Name=dyn_headset_02, swing1LimitDegrees=3, swing2LimitDegrees=3, twistLimitDegrees=1}
]
